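# Deformation Control Rig for the TrendyPunk female medium player body, adding outfit-specific jacket fixes after the shared deform rig function.
# The lapel function offsets lapel and collar bones, weighted by clavicle swing-twist, spherical upper-arm pose readers, an animation attribute, and hand-gliding, pistol and emote variables.
# A constraint function then binds three pin bones to weighted spine and jacket bones; its right-lapel constraints are unwired.
# An FK control preview pass is switched off.
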ControlRigBlueprint /BRCosmetics/Characters/Player/Female/Medium/Bodies/F_MED_TrendyPunk/Meshes/F_MED_TrendyPunk_Deform_CtrlRig
graphs [
 New Function_ContainedGraph {nodes [
   RigVMFunctionEntryNode()
   > RigVMFunction_Sequence = RigVMAggregateNode()
   &RigUnit_RotationConstraintLocalSpaceOffset(
    AdvancedSettings=(InterpolationType=Average,RotationOrderForFilter=XZY)
    Child=(Type=Bone,Name="base_dyn_JacketLapels_r_01")
    Filter=(bX=True,bY=True,bZ=True)
    Parents=((Item=(Type=Bone,Name="clavicle_r"),Weight=0.8),(Item=(Type=Bone,Name="upperarm_r"),Weight=0.2))
    Weight=1
    ExecutePin=@RigVMFunction_Sequence.A
   )
   > DeltaRotation = RigVMCollapseNode(Rotation Order=ZYX, child=clavicle_l, max=-40, parent=clavicle_l_null, parentType=Null, premult=1, primAxis=1, twistAxis=(X=1,Y=0,Z=0))
   > RigVMVariableNode(Value=@DeltaRotation.outputWeight) {variable=isClavUp_L}
   > DeltaRotation_1 = RigVMCollapseNode(Rotation Order=ZYX, child=clavicle_r, max=-40, parent=clavicle_r_null, parentType=Null, premult=1, primAxis=1, twistAxis=(X=1,Y=0,Z=0))
   > RigVMVariableNode(Value=@DeltaRotation_1.outputWeight) {variable=isClavUp_R}
   RigVMFunction_MathBoolToFloat = &RigVMFunction_MathBoolToFloat(Value=RigVMVariableNode() {variable=IshandglaidingActice}.Value)
   RigVMFunction_AnimEvalRichCurve = &RigVMFunction_AnimEvalRichCurve(
    Curve=(EditorCurveData=(Keys=((),(InterpMode=RCIM_Cubic,Time=0.4995,Value=0.4,ArriveTangent=1,LeaveTangent=1),(InterpMode=RCIM_Cubic,TangentMode=RCTM_Break,Time=1,Value=1,ArriveTangent=1.3135))))
    SourceMinimum=0
    TargetMinimum=0
    Value=@RigVMFunction_MathBoolToFloat.Result
   )
   &RigUnit_ModifyTransforms(
    ItemToModify="((Item=(Type=Bone,Name=\"base_dyn_backCollar\"),Transform=(Rotation=(X=0,Y=0.5,Z=0,W=0.866),Translation=(X=0,Y=0,Z=-3),Scale3D=(X=1,Y=1,Z=1))),(Item=(Type=Bone,Name=\"base_dyn_JacketLapels_l_01\"),Transform=(Rotation=(X=0.0864,Y=0.13,Z=-0.0114,W=0.9877),Transl ...(truncated)"
    WeightMinimum=0
    Weight=@RigVMFunction_AnimEvalRichCurve.Result
    ExecutePin=@RigVMFunction_Sequence.B
   )
   > &RigUnit_ModifyTransforms(
    ItemToModify="((Item=(Type=Bone,Name=\"base_dyn_JacketLapels_l_01\"),Transform=(Rotation=(X=0,Y=0,Z=0,W=1),Translation=(X=0,Y=1,Z=0),Scale3D=(X=1,Y=1,Z=1))),(Item=(Type=Bone,Name=\"base_dyn_JacketLapels_Mid_l_02\"),Transform=(Rotation=(X=0,Y=0,Z=0,W=1),Translation=(X=0,Y=5, ...(truncated)"
    WeightMinimum=0
    Weight=RigVMVariableNode() {variable=isClavUp_L}.Value
   )
   > &RigUnit_ModifyTransforms(
    ItemToModify="((Item=(Type=Bone,Name=\"base_dyn_JacketLapels_r_01\"),Transform=(Rotation=(X=0.0396,Y=-0.0889,Z=-0.0396,W=0.9945),Translation=(X=0,Y=0,Z=0),Scale3D=(X=1,Y=1,Z=1))),(Item=(Type=Bone,Name=\"base_dyn_JacketLapels_Mid_r_02\"),Transform=(Rotation=(X=0,Y=-0.1736,Z= ...(truncated)"
    WeightMinimum=0
    Weight=RigVMVariableNode() {variable=isClavUp_R}.Value
   )
   Get_Animation_Attribute = RigVMDispatchNode(BoneName=root, Default=0, Name=deformRig_r_upperarm_fr_alpha) {template="DISPATCH_RigDispatch_GetAnimAttribute(in Name,in BoneName,in Default,out Value,out Found)"}
   > &RigUnit_ModifyTransforms(
    ItemToModify="((Item=(Type=Bone,Name=\"base_dyn_JacketLapels_r_01\"),Transform=(Rotation=(X=-0.1751,Y=0.1272,Z=-0.5739,W=0.7898),Translation=(X=0,Y=0,Z=0),Scale3D=(X=1,Y=1,Z=1))),(Item=(Type=Bone,Name=\"base_JacketLapels_r_02\"),Transform=(Rotation=(X=-0.0182,Y=-0.166,Z=0.0 ...(truncated)"
    WeightMinimum=0
    Weight=@Get_Animation_Attribute.Value
   )
   > n1 = &RigUnit_SphericalPoseReader(
    ActiveRegionScaleFactors=(PositiveWidth=0.8,NegativeWidth=0.8,PositiveHeight=0.8,NegativeHeight=0.8)
    ActiveRegionSize=0.1
    Debug=(bDrawDebug=false,bDraw2D=False,bDrawLocalAxes=False,DebugScale=25,DebugSegments=20,DebugThickness=0.25)
    DriverAxis=(X=1,Y=0,Z=0)
    DriverItem=(Type=Bone,Name="upperarm_l")
    FalloffRegionScaleFactors=(PositiveWidth=1,NegativeWidth=1,PositiveHeight=1,NegativeHeight=1)
    FalloffSize=0.15
    FlipHeightScaling=False
    FlipWidthScaling=False
    OptionalParentItem=(Type=Bone,Name="None")
    RotationOffset=(X=45,Y=-50,Z=0)
   )
   > &RigUnit_ModifyTransforms(
    ItemToModify="((Item=(Type=Bone,Name=\"base_dyn_JacketLapels_l_01\"),Transform=(Rotation=(X=0,Y=0,Z=0.2588,W=0.9659),Translation=(X=0,Y=1,Z=0),Scale3D=(X=1,Y=1,Z=1))),(Item=(Type=Bone,Name=\"base_dyn_JacketLapels_Mid_l_02\"),Transform=(Rotation=(X=0,Y=0,Z=0,W=1),Translation ...(truncated)"
    WeightMinimum=0
    Weight=@n1.OutputParam
   )
   > n2 = &RigUnit_SphericalPoseReader(
    ActiveRegionScaleFactors=(PositiveWidth=0.8,NegativeWidth=0.8,PositiveHeight=0.5,NegativeHeight=1)
    ActiveRegionSize=0.1
    Debug=(bDrawDebug=false,bDraw2D=False,bDrawLocalAxes=False,DebugScale=25,DebugSegments=20,DebugThickness=0.25)
    DriverAxis=(X=-1,Y=0,Z=0)
    DriverItem=(Type=Bone,Name="upperarm_r")
    FalloffRegionScaleFactors=(PositiveWidth=1,NegativeWidth=1,PositiveHeight=1.5,NegativeHeight=1.5)
    FalloffSize=0.15
    FlipHeightScaling=False
    FlipWidthScaling=False
    OptionalParentItem=(Type=Bone,Name="None")
    RotationOffset=(X=-105,Y=0,Z=20)
   )
   Multiply = &RigVMFunction_MathDoubleMul(B=&RigVMFunction_MathBoolToFloat(Value=RigVMVariableNode() {variable=isPistol}.Value).Result, A=@n2.OutputParam)
   > &RigUnit_ModifyTransforms(
    ItemToModify="((Item=(Type=Bone,Name=\"base_dyn_JacketLapels_r_01\"),Transform=(Rotation=(X=0.2364,Y=-0.2019,Z=-0.3778,W=0.8721),Translation=(X=0,Y=0,Z=0),Scale3D=(X=1,Y=1,Z=1))),(Item=(Type=Bone,Name=\"base_JacketLapels_r_02\"),Transform=(Rotation=(X=0,Y=0,Z=0,W=1),Transla ...(truncated)"
    WeightMinimum=0
    Weight=@Multiply.Result
   )
   &RigUnit_ModifyTransforms(
    ItemToModify=((Item=(Type=Bone,Name="base_JacketLapels_r_02"),Transform=(Rotation=(X=0,Y=-0.2164,Z=0,W=0.9763),Translation=(X=0,Y=0,Z=2),Scale3D=(X=1,Y=1,Z=1))))
    WeightMinimum=0
    ExecutePin=@RigVMFunction_Sequence.C
    Weight=RigVMVariableNode() {variable=IsEmoteActive}.Value
   )
   RigVMFunctionReturnNode(ExecuteContext=@RigVMFunction_Sequence.D)
  ]}
 New Function_ContainedGraph {nodes [
   n1 = RigVMFunctionEntryNode()
   RigVMFunction_MathBoolToInteger = &RigVMFunction_MathBoolToInteger(Value=@n1.onlyTwist)
   RigVMFunction_MathQuaternionSwingTwist = &RigVMFunction_MathQuaternionSwingTwist(
    TwistAxis=@n1.twistAxis
    Input=&RigUnit_GetRelativeTransformForItem(
     Child=(Type=Bone,Name="None")
     Parent=(Type=Bone,Name="None")
     bChildInitial=False
     bParentInitial=False
     Child.Name=@n1.child
     Parent.Name=@n1.parent
     Parent.Type=@n1.parentType
    ).RelativeTransform.Rotation
   )
   RigVMFunction_MathQuaternionToEuler = &RigVMFunction_MathQuaternionToEuler(
    Value=RigVMDispatchNode(
     Values=((),())
     Index=@RigVMFunction_MathBoolToInteger.Result
     Values.0=@RigVMFunction_MathQuaternionSwingTwist.Swing
     Values.1=@RigVMFunction_MathQuaternionSwingTwist.Twist
    ) {template="DISPATCH_RigVMDispatch_SelectInt32(in Index,in Values,out Result)"}.Result
    RotationOrder=@n1.Rotation Order
   )
   > DISPATCH_RigVMDispatch_Print = RigVMDispatchNode(ScreenColor=(R=1,G=1,B=1,A=1), ScreenDuration=1.5, Enabled=@n1.debugRotation, Value=@RigVMFunction_MathQuaternionToEuler.Result) {template="DISPATCH_RigVMDispatch_Print(in Prefix,in Value,in Enabled,in ScreenDuration,in ScreenColor)"}
   Clamp = &RigVMFunction_MathIntClamp(Value=@n1.primAxis)
   DISPATCH_RigVMDispatch_SelectInt32 = RigVMDispatchNode(
    Values=(0,0,0)
    Index=@Clamp.Result
    Values.0=@RigVMFunction_MathQuaternionToEuler.Result.X
    Values.1=@RigVMFunction_MathQuaternionToEuler.Result.Y
    Values.2=@RigVMFunction_MathQuaternionToEuler.Result.Z
   ) {template="DISPATCH_RigVMDispatch_SelectInt32(in Index,in Values,out Result)"}
   Remap = &RigVMFunction_MathDoubleRemap(
    TargetMaximum=1
    TargetMinimum=0
    bClamp=False
    Value=@DISPATCH_RigVMDispatch_SelectInt32.Result
    SourceMaximum=@n1.max
    SourceMinimum=@n1.min
   )
   Multiply = &RigVMFunction_MathDoubleMul(A=&RigVMFunction_MathDoubleClamp(Value=@Remap.Result).Result, B=@n1.premult)
   > RigVMFunctionReturnNode(outputWeight=@Multiply.Result)
  ]}
 New Function_ContainedGraph {nodes [
   n1 = RigVMFunctionEntryNode()
   RigVMFunction_MathBoolToInteger = &RigVMFunction_MathBoolToInteger(Value=@n1.onlyTwist)
   RigVMFunction_MathQuaternionSwingTwist = &RigVMFunction_MathQuaternionSwingTwist(
    TwistAxis=@n1.twistAxis
    Input=&RigUnit_GetRelativeTransformForItem(
     Child=(Type=Bone,Name="None")
     Parent=(Type=Bone,Name="None")
     bChildInitial=False
     bParentInitial=False
     Child.Name=@n1.child
     Parent.Name=@n1.parent
     Parent.Type=@n1.parentType
    ).RelativeTransform.Rotation
   )
   RigVMFunction_MathQuaternionToEuler = &RigVMFunction_MathQuaternionToEuler(
    Value=RigVMDispatchNode(
     Values=((),())
     Index=@RigVMFunction_MathBoolToInteger.Result
     Values.0=@RigVMFunction_MathQuaternionSwingTwist.Swing
     Values.1=@RigVMFunction_MathQuaternionSwingTwist.Twist
    ) {template="DISPATCH_RigVMDispatch_SelectInt32(in Index,in Values,out Result)"}.Result
    RotationOrder=@n1.Rotation Order
   )
   > DISPATCH_RigVMDispatch_Print = RigVMDispatchNode(ScreenColor=(R=1,G=1,B=1,A=1), ScreenDuration=1.5, Enabled=@n1.debugRotation, Value=@RigVMFunction_MathQuaternionToEuler.Result) {template="DISPATCH_RigVMDispatch_Print(in Prefix,in Value,in Enabled,in ScreenDuration,in ScreenColor)"}
   Clamp = &RigVMFunction_MathIntClamp(Value=@n1.primAxis)
   DISPATCH_RigVMDispatch_SelectInt32 = RigVMDispatchNode(
    Values=(0,0,0)
    Index=@Clamp.Result
    Values.0=@RigVMFunction_MathQuaternionToEuler.Result.X
    Values.1=@RigVMFunction_MathQuaternionToEuler.Result.Y
    Values.2=@RigVMFunction_MathQuaternionToEuler.Result.Z
   ) {template="DISPATCH_RigVMDispatch_SelectInt32(in Index,in Values,out Result)"}
   Remap = &RigVMFunction_MathDoubleRemap(
    TargetMaximum=1
    TargetMinimum=0
    bClamp=False
    Value=@DISPATCH_RigVMDispatch_SelectInt32.Result
    SourceMaximum=@n1.max
    SourceMinimum=@n1.min
   )
   Multiply = &RigVMFunction_MathDoubleMul(A=&RigVMFunction_MathDoubleClamp(Value=@Remap.Result).Result, B=@n1.premult)
   > RigVMFunctionReturnNode(outputWeight=@Multiply.Result)
  ]}
 New Function_ContainedGraph {nodes [
   RigVMFunctionEntryNode()
   > &RigUnit_ParentConstraint(
    AdvancedSettings=(RotationOrderForFilter=XZY)
    Child=(Type=Bone,Name="base_dyn_pin_02")
    Parents="((Item=(Type=Bone,Name=\"dyn_front_jacket_r\"),Weight=0.0056),(Item=(Type=Bone,Name=\"spine_02\"),Weight=0.0138),(Item=(Type=Bone,Name=\"dyn_JacketLapels_r_01\"),Weight=0.0274),(Item=(Type=Bone,Name=\"dyn_JacketLapels_r_02\"),Weight=0.0303),(Item=(Type=Bone,Name=\" ...(truncated)"
    Weight=1
   )
   > &RigUnit_ParentConstraint(
    AdvancedSettings=(RotationOrderForFilter=XZY)
    Child=(Type=Bone,Name="base_dyn_pin_03")
    Parents="((Item=(Type=Bone,Name=\"dyn_front_jacket_l\"),Weight=0.0004),(Item=(Type=Bone,Name=\"dyn_side_jacket_l\"),Weight=0.0006),(Item=(Type=Bone,Name=\"spine_03\"),Weight=0.0037),(Item=(Type=Bone,Name=\"spine_02\"),Weight=0.0485),(Item=(Type=Bone,Name=\"spine_05\"),Weight ...(truncated)"
    Weight=1
   )
   > &RigUnit_ParentConstraint(
    AdvancedSettings=(RotationOrderForFilter=XZY)
    Child=(Type=Bone,Name="base_dyn_pin_01")
    Parents="((Item=(Type=Bone,Name=\"dyn_side_jacket_r\"),Weight=0.0309),(Item=(Type=Bone,Name=\"spine_02\"),Weight=0.0521),(Item=(Type=Bone,Name=\"spine_05\"),Weight=0.0591),(Item=(Type=Bone,Name=\"dyn_front_jacket_r\"),Weight=0.1046),(Item=(Type=Bone,Name=\"spine_04\"),Weight ...(truncated)"
    Weight=1
   )
   > RigVMFunctionReturnNode()
   &RigUnit_ParentConstraint(
    AdvancedSettings=(RotationOrderForFilter=XZY)
    Child=(Type=Bone,Name="base_dyn_JacketLapels_r_01")
    ExecutePin=()
    Parents="((Item=(Type=Bone,Name=\"clavicle_l\"),Weight=0.0055),(Item=(Type=Bone,Name=\"dyn_backCollar\"),Weight=0.0097),(Item=(Type=Bone,Name=\"dyn_JacketLapels_Mid_r_02\"),Weight=0.0143),(Item=(Type=Bone,Name=\"spine_04\"),Weight=0.0496),(Item=(Type=Bone,Name=\"dyn_JacketL ...(truncated)"
    Weight=1
   )
   > &RigUnit_ParentConstraint(
    AdvancedSettings=(RotationOrderForFilter=XZY)
    Child=(Type=Bone,Name="base_JacketLapels_r_02")
    Parents="((Item=(Type=Bone,Name=\"clavicle_l\"),Weight=0.0006),(Item=(Type=Bone,Name=\"dyn_backCollar\"),Weight=0.0014),(Item=(Type=Bone,Name=\"spine_04\"),Weight=0.006),(Item=(Type=Bone,Name=\"clavicle_r\"),Weight=0.0149),(Item=(Type=Bone,Name=\"spine_05\"),Weight=0.0482),( ...(truncated)"
    Weight=1
   )
   > &RigUnit_ParentConstraint(
    AdvancedSettings=(RotationOrderForFilter=XZY)
    Child=(Type=Bone,Name="base_dyn_JacketLapels_Mid_r_02")
    Parents="((Item=(Type=Bone,Name=\"deform_upperarm_fr_r\"),Weight=0.0003),(Item=(Type=Bone,Name=\"deform_pec_r\"),Weight=0.0043),(Item=(Type=Bone,Name=\"dyn_JacketLapels_r_02\"),Weight=0.0132),(Item=(Type=Bone,Name=\"dyn_JacketLapels_r_01\"),Weight=0.0341),(Item=(Type=Bone, ...(truncated)"
    Weight=1
   )
  ]}
 New Function_ContainedGraph {nodes [
   n1 = RigVMFunctionEntryNode()
   RigVMFunction_ControlFlowBranch = &RigVMFunction_ControlFlowBranch(ExecuteContext=@n1.Execute, Condition=@n1.On)
   DISPATCH_RigVMDispatch_ArrayIterator = RigVMDispatchNode(
    Array=/Script/ControlRig.RigUnit_CollectionGetAll(TypeToSearch=Control) {is=struct RigVMUnitNode}.Items
    ExecuteContext=@RigVMFunction_ControlFlowBranch.True
   ) {template="DISPATCH_RigVMDispatch_ArrayIterator(in Array,out Element,out Index,out Count,out Ratio)"}
   Replace = /Script/RigVM.RigVMFunction_NameReplace(Old=_ctrl, Name=@DISPATCH_RigVMDispatch_ArrayIterator.Element.Name) {is=struct RigVMUnitNode, template="Replace::Execute(in Name,in Old,in New,out Result)"}
   > &RigUnit_SetTransform(
    Item=(Type=Bone,Name="spine_01")
    Weight=1
    bInitial=False
    Item.Name=@Replace.Result
    Value=&RigUnit_GetTransform(Item=(Type=Control,Name="spine_01_ctrl"), bInitial=False, Item.Name=@DISPATCH_RigVMDispatch_ArrayIterator.Element.Name).Transform
   )
   RigVMFunctionReturnNode(Execute=@RigVMFunction_ControlFlowBranch.Completed)
  ]}
 RigVMFunctionLibrary {nodes [
   ConnectFKPreviewControls = RigVMCollapseNode(Execute=(), On=False)
   ConstrainsFuntion = RigVMCollapseNode(ExecuteContext=())
   LapelsFuntion = RigVMCollapseNode(ExecuteContext=())
  ]}
 RigVMFunction_Sequence_1_ContainedGraph {nodes [
   n1 = RigVMFunctionEntryNode()
   RigVMFunction_Sequence = &RigVMFunction_Sequence(ExecuteContext=@n1.ExecuteContext)
   n2 = &RigVMFunction_Sequence(ExecuteContext=@RigVMFunction_Sequence.B)
   n3 = &RigVMFunction_Sequence(ExecuteContext=@n2.B)
   RigVMFunctionReturnNode(A=@RigVMFunction_Sequence.A, B=@n2.A, C=@n3.A, D=@n3.B)
  ]}
 RigVMModel {nodes [
   RigUnit_BeginExecution = &RigUnit_BeginExecution()
   ConnectFKPreviewControls = RigVMFunctionReferenceNode(On=false, Execute=@RigUnit_BeginExecution.ExecutePin)
   FN_DeformRig_FNC = RigVMFunctionReferenceNode(
    DeformRig_ Curve="(deformRig_l_upperarm_up_alpha_72_#1=False,deformRig_r_upperarm_up_alpha_73_#2=False,deformRig_l_upperarm_fr_alpha_74_#3=true,deformRig_r_upperarm_fr_alpha_75_CB51B80 ...(truncated)"
    DeformRig_Bool="(upperarm_1_#4=True,elbow_3_#5=True,wrist_7_#6=True,groin_8_#7=True,glute_17_#8=True,knee_9_40C408324CA261 ...(truncated)"
    DeformRig_Debug="(upperarm_debug=False,wrist_debug=False,lat_debug=False,groinGlute_debug=False,upperarm_debug_18_#4=False,wrist_debug_20_#6=False,groinGlute_debug_27_#7=False,lat_deb ...(truncated)"
    DeformRig_Value="(upperarm_up_18_#9=(Translation=(X=0,Y=0,Z=8),Scale3D=(X=0.7,Y=1,Z=1)),upperarm_fwd_fr_19_#10=(Translation=(X=0,Y=10,Z=0),Scale3D=(X=0.5,Y=1,Z=1)),upperarm_fwd_bk_20_9973AAD84EE0A1E349BC9E85067D4A3 ...(truncated)"
    ExecuteContext=@ConnectFKPreviewControls.Execute
   )
   > LapelsFuntion = RigVMFunctionReferenceNode()
   > ConstrainsFuntion = RigVMFunctionReferenceNode()
  ]}
]
hierarchy [
 root {loc=<0 0 0>, rot=<0 0 0 1>}
 pelvis {parent=root, loc=<0 -0.5301 98.0571>, rot=<0.0127 0.707 -0.0127 -0.707>}
 spine_01 {parent=pelvis, loc=<6.1208 -0.9328 0>, rot=<0 0 0.034 -0.9994>}
 spine_02 {parent=spine_01, loc=<6.5997 -0.2095 0>, rot=<0 0 0.0238 -0.9997>}
 spine_03 {parent=spine_02, loc=<6.4914 -0.1237 0>, rot=<0 0 0.0149 -0.9999>}
 spine_04 {parent=spine_03, loc=<6.9598 0.0794 0>, rot=<0 0 -0.0132 -0.9999>}
 spine_05 {parent=spine_04, loc=<6.9987 0 0>, rot=<0 0 -0.0015 -1>}
 neck_01 {parent=spine_05, loc=<11.3138 0.1529 0>, rot=<0 0 -0.2003 -0.9797>}
 neck_02 {parent=neck_01, loc=<4.5627 0 0>, rot=<0 0 0.0504 -0.9987>}
 head {parent=neck_02, loc=<5.3914 0 0>, rot=<0 0 0.115 -0.9934>}
 clavicle_l {parent=spine_05, loc=<5.4831 -0.1053 -3.2946>, rot=<0.0465 -0.7289 -0.009 -0.6829>}
 upperarm_l {parent=clavicle_l, loc=<9.8658 -1.2136 -0.4105>, rot=<0.0164 -0.3245 -0.0269 -0.9454>}
 lowerarm_l {parent=upperarm_l, loc=<26.465 -0.8345 0.259>, rot=<0.0036 0.0161 -0.2298 -0.9731>}
 hand_l {parent=lowerarm_l, loc=<22.7355 -0.3203 -0.6105>, rot=<-0.6784 -0.0223 0.0735 -0.7307>}
 thumb_01_l {parent=hand_l, loc=<1.3118 -0.3191 -2.459>, rot=<0.6546 -0.4042 0.0664 -0.6354>}
 thumb_02_l {parent=thumb_01_l, loc=<5.9336 -0.0448 0.0401>, rot=<-0.0037 0.0213 0.298 -0.9543>}
 thumb_03_l {parent=thumb_02_l, loc=<2.2223 0.3682 -0.0907>, rot=<0.0118 -0.013 0.0066 -0.9998>}
 index_metacarpal_l {parent=hand_l, loc=<1.6535 0.0183 -2.96>, rot=<0.0158 -0.0124 -0.0173 -0.9996>}
 index_01_l {parent=index_metacarpal_l, loc=<8.3008 0.0357 -0.0524>, rot=<-0.0044 0.0485 0.0924 -0.9945>}
 index_02_l {parent=index_01_l, loc=<3.3587 -0.3004 0.0452>, rot=<-0.002 0.0027 0.3225 -0.9466>}
 index_03_l {parent=index_02_l, loc=<2.5578 -0.123 -0.0252>, rot=<-0.002 0.0067 0.2558 -0.9667>}
 middle_metacarpal_l {parent=hand_l, loc=<2.1856 -0.1725 -0.5984>, rot=<0.006 0.0168 -0.0329 -0.9993>}
 middle_01_l {parent=middle_metacarpal_l, loc=<7.3275 -0.2447 0.0129>, rot=<-0.0208 0.0042 0.0858 -0.9961>}
 middle_02_l {parent=middle_01_l, loc=<4.0123 0.1401 0.1175>, rot=<-0.0218 0.0258 0.3512 -0.9357>}
 middle_03_l {parent=middle_02_l, loc=<2.6077 0.0534 0.1373>, rot=<0.0082 0.0188 0.1708 -0.9851>}
 ring_metacarpal_l {parent=hand_l, loc=<2.2074 -0.3625 1.3638>, rot=<-0.0284 0.0722 -0.0458 -0.9959>}
 ring_01_l {parent=ring_metacarpal_l, loc=<7.0295 -0.6197 -0.1446>, rot=<-0.0489 -0.0252 0.1892 -0.9804>}
 ring_02_l {parent=ring_01_l, loc=<3.6174 0.4647 -0.1027>, rot=<0 0 0.2806 -0.9598>}
 ring_03_l {parent=ring_02_l, loc=<2.4152 0.114 0.0199>, rot=<0 0 0.1189 -0.9929>}
 pinky_metacarpal_l {parent=hand_l, loc=<1.7616 -0.9753 3.3814>, rot=<-0.0397 0.0814 -0.069 -0.9935>}
 pinky_01_l {parent=pinky_metacarpal_l, loc=<7.1709 -0.4831 -0.1094>, rot=<-0.0654 -0.0035 0.2242 -0.9723>}
 pinky_02_l {parent=pinky_01_l, loc=<2.8274 0.0893 -0.1043>, rot=<0.0037 0.0011 0.2854 -0.9584>}
 pinky_03_l {parent=pinky_02_l, loc=<2.2046 0.0307 -0.0078>, rot=<-0.0002 0.0098 0.1445 -0.9895>}
 deform_wrist_dn_base_l {parent=hand_l, loc=<0 -1.5326 0>, rot=<0 0 0 -1>}
 deform_wrist_dn_l {parent=deform_wrist_dn_base_l, loc=<0 0 0>, rot=<0 0 0 1>}
 deform_wrist_up_base_l {parent=hand_l, loc=<0 2.3009 0>, rot=<0 0 0 -1>}
 deform_wrist_up_l {parent=deform_wrist_up_base_l, loc=<0 0 0>, rot=<0 0 0 1>}
 lowerarm_twist_02_l {parent=lowerarm_l, loc=<10.6707 0 0>, rot=<0 0 0 1>}
 lowerarm_twist_01_l {parent=lowerarm_l, loc=<18.316 0 0>, rot=<0 0 0 1>}
 deform_elbow_in_base_l {parent=lowerarm_l, loc=<0.1834 3.6062 -0.2572>, rot=<0 0 0 -1>}
 deform_elbow_in_l {parent=deform_elbow_in_base_l, loc=<0 0 0>, rot=<0 0 0 1>}
 deform_elbow_out_base_l {parent=lowerarm_l, loc=<0 -4 0>, rot=<0 0 0 1>}
 deform_elbow_out_l {parent=deform_elbow_out_base_l, loc=<0 0 0>, rot=<0 0 0 1>}
 base_JacketSleeve_l {parent=lowerarm_l, loc=<4.099 -0.045 -0.07>, rot=<0.1324 0.0084 0.0571 -0.9895>}
 dyn_JacketSleeve_a_l {parent=base_JacketSleeve_l, loc=<1.3784 -0.1554 1.1918>, rot=<-0.0253 0.2253 0.0613 -0.972>}
 dyn_JacketSleeve_b_l {parent=base_JacketSleeve_l, loc=<1.2983 1.267 0.8006>, rot=<0.1241 0.0865 -0.2121 -0.9655>}
 dyn_JacketSleeve_c_l {parent=base_JacketSleeve_l, loc=<1.4406 1.0088 -0.7834>, rot=<0.1835 -0.1723 -0.179 -0.9511>}
 dyn_JacketSleeve_d_l {parent=base_JacketSleeve_l, loc=<1.0952 -1.039 -0.3275>, rot=<0.9628 -0.2232 0.0662 -0.1368>}
 upperarm_twist_01_l {parent=upperarm_l, loc=<7.2135 -0.1134 0>, rot=<0 0 0 1>}
 deform_upperarm_bk_base_l {parent=upperarm_twist_01_l, loc=<-6.4053 -4.3746 0.314>, rot=<0 0 0 -1>}
 deform_upperarm_bk_l {parent=deform_upperarm_bk_base_l, loc=<0 0 0>, rot=<0 0 0 1>}
 deform_upperarm_fr_base_l {parent=upperarm_twist_01_l, loc=<-5.6312 5.092 0.3354>, rot=<0 0 0 -1>}
 deform_upperarm_fr_l {parent=deform_upperarm_fr_base_l, loc=<0 0 0>, rot=<0 0 0 1>}
 deform_upperarm_up_base_l {parent=upperarm_twist_01_l, loc=<-8.891 0.4825 3.9517>, rot=<0 0 0 -1>}
 deform_upperarm_up_l {parent=deform_upperarm_up_base_l, loc=<0 0 0>, rot=<0 0 0 1>}
 upperarm_twist_02_l {parent=upperarm_l, loc=<19.7998 -0.1134 0>, rot=<0 0 0 1>}
 base_dyn_JacketLapels_l_01 {parent=clavicle_l, loc=<4.2355 2.0018 5.4811>, rot=<0.0053 0.0211 -0.1144 -0.9932>}
 dyn_JacketLapels_l_01 {parent=base_dyn_JacketLapels_l_01, loc=<0 0 0>, rot=<0 0 0 1>}
 base_JacketLapels_l_02 {parent=dyn_JacketLapels_l_01, loc=<7.7017 0 0>, rot=<0.0103 -0.1241 -0.0687 -0.9898>}
 dyn_JacketLapels_l_02 {parent=base_JacketLapels_l_02, loc=<0 0 0>, rot=<0 0 0 1>}
 clavicle_r {parent=spine_05, loc=<5.4831 -0.1053 3.2946>, rot=<0.7289 0.0465 -0.6829 0.009>}
 upperarm_r {parent=clavicle_r, loc=<-9.8658 1.2136 0.4105>, rot=<0.0164 -0.3245 -0.0269 -0.9454>}
 lowerarm_r {parent=upperarm_r, loc=<-26.465 0.8345 -0.259>, rot=<0.0036 0.0161 -0.2298 -0.9731>}
 hand_r {parent=lowerarm_r, loc=<-22.7355 0.3203 0.6106>, rot=<-0.6784 -0.0223 0.0735 -0.7307>}
 thumb_01_r {parent=hand_r, loc=<-1.3118 0.3191 2.459>, rot=<0.6546 -0.4042 0.0664 -0.6354>}
 thumb_02_r {parent=thumb_01_r, loc=<-5.9336 0.0448 -0.0401>, rot=<-0.0037 0.0213 0.298 -0.9543>}
 thumb_03_r {parent=thumb_02_r, loc=<-2.2223 -0.3682 0.0907>, rot=<0.0118 -0.013 0.0066 -0.9998>}
 index_metacarpal_r {parent=hand_r, loc=<-1.6535 -0.0183 2.96>, rot=<0.0158 -0.0124 -0.0173 -0.9996>}
 index_01_r {parent=index_metacarpal_r, loc=<-8.3008 -0.0357 0.0524>, rot=<-0.0044 0.0485 0.0924 -0.9945>}
 index_02_r {parent=index_01_r, loc=<-3.3587 0.3004 -0.0452>, rot=<-0.002 0.0027 0.3225 -0.9466>}
 index_03_r {parent=index_02_r, loc=<-2.5578 0.123 0.0252>, rot=<-0.002 0.0067 0.2558 -0.9667>}
 middle_metacarpal_r {parent=hand_r, loc=<-2.1856 0.1725 0.5984>, rot=<0.006 0.0168 -0.0329 -0.9993>}
 middle_01_r {parent=middle_metacarpal_r, loc=<-7.3275 0.2447 -0.0129>, rot=<-0.0208 0.0042 0.0858 -0.9961>}
 middle_02_r {parent=middle_01_r, loc=<-4.0123 -0.1401 -0.1175>, rot=<-0.0218 0.0258 0.3512 -0.9357>}
 middle_03_r {parent=middle_02_r, loc=<-2.6077 -0.0534 -0.1373>, rot=<0.0082 0.0188 0.1708 -0.9851>}
 ring_metacarpal_r {parent=hand_r, loc=<-2.2074 0.3625 -1.3638>, rot=<-0.0284 0.0722 -0.0458 -0.9959>}
 ring_01_r {parent=ring_metacarpal_r, loc=<-7.0295 0.6197 0.1446>, rot=<-0.0489 -0.0252 0.1892 -0.9804>}
 ring_02_r {parent=ring_01_r, loc=<-3.6174 -0.4647 0.1027>, rot=<0 0 0.2806 -0.9598>}
 ring_03_r {parent=ring_02_r, loc=<-2.4152 -0.114 -0.0199>, rot=<0 0 0.1189 -0.9929>}
 pinky_metacarpal_r {parent=hand_r, loc=<-1.7616 0.9753 -3.3814>, rot=<-0.0397 0.0814 -0.069 -0.9935>}
 pinky_01_r {parent=pinky_metacarpal_r, loc=<-7.1709 0.4831 0.1094>, rot=<-0.0654 -0.0035 0.2242 -0.9723>}
 pinky_02_r {parent=pinky_01_r, loc=<-2.8274 -0.0893 0.1043>, rot=<0.0037 0.0011 0.2854 -0.9584>}
 pinky_03_r {parent=pinky_02_r, loc=<-2.2046 -0.0307 0.0078>, rot=<-0.0002 0.0098 0.1445 -0.9895>}
 deform_wrist_dn_base_r {parent=hand_r, loc=<-0.0001 1.5328 0>, rot=<0 0 0 -1>}
 deform_wrist_dn_r {parent=deform_wrist_dn_base_r, loc=<0 0 0>, rot=<0 0 0 1>}
 deform_wrist_up_base_r {parent=hand_r, loc=<-0.0002 -2.3007 -0.0001>, rot=<0 0 0 -1>}
 deform_wrist_up_r {parent=deform_wrist_up_base_r, loc=<0 0 0>, rot=<0 0 0 1>}
 lowerarm_twist_02_r {parent=lowerarm_r, loc=<-10.6707 0 0>, rot=<0 0 0 -1>}
 lowerarm_twist_01_r {parent=lowerarm_r, loc=<-18.316 0 0>, rot=<0 0 0 -1>}
 deform_elbow_in_base_r {parent=lowerarm_r, loc=<-0.1834 -3.6062 0.2572>, rot=<0 0 0 -1>}
 deform_elbow_in_r {parent=deform_elbow_in_base_r, loc=<0 0 0>, rot=<0 0 0 1>}
 deform_elbow_out_base_r {parent=lowerarm_r, loc=<0 4 -0.0001>, rot=<0 0 0 -1>}
 deform_elbow_out_r {parent=deform_elbow_out_base_r, loc=<0 0 0>, rot=<0 0 0 1>}
 base_JacketSleeve_r {parent=lowerarm_r, loc=<-4.0992 0.0451 0.0702>, rot=<0.1324 0.0084 0.0571 -0.9895>}
 dyn_JacketSleeve_a_r {parent=base_JacketSleeve_r, loc=<-1.3781 0.1553 -1.1921>, rot=<0.0436 -0.9705 0.0174 -0.2364>}
 dyn_JacketSleeve_b_r {parent=base_JacketSleeve_r, loc=<-1.2979 -1.267 -0.8012>, rot=<0.239 -0.9553 0.1736 0.0093>}
 dyn_JacketSleeve_c_r {parent=base_JacketSleeve_r, loc=<-1.4403 -1.0088 0.7828>, rot=<0.1842 -0.9487 0.1895 0.1736>}
 dyn_JacketSleeve_d_r {parent=base_JacketSleeve_r, loc=<-1.0952 1.039 0.3276>, rot=<-0.218 -0.9611 -0.1666 0.0327>}
 upperarm_twist_01_r {parent=upperarm_r, loc=<-7.2135 0.1134 0>, rot=<0 0 0 -1>}
 deform_upperarm_bk_base_r {parent=upperarm_twist_01_r, loc=<6.4053 4.3746 -0.314>, rot=<0 0 0 -1>}
 deform_upperarm_bk_r {parent=deform_upperarm_bk_base_r, loc=<0 0 0>, rot=<0 0 0 1>}
 deform_upperarm_fr_base_r {parent=upperarm_twist_01_r, loc=<5.6312 -5.092 -0.3354>, rot=<0 0 0 -1>}
 deform_upperarm_fr_r {parent=deform_upperarm_fr_base_r, loc=<0 0 0>, rot=<0 0 0 1>}
 deform_upperarm_up_base_r {parent=upperarm_twist_01_r, loc=<8.891 -0.4825 -3.9517>, rot=<0 0 0 -1>}
 deform_upperarm_up_r {parent=deform_upperarm_up_base_r, loc=<0 0 0>, rot=<0 0 0 1>}
 upperarm_twist_02_r {parent=upperarm_r, loc=<-19.7998 0.1134 0>, rot=<0 0 0 -1>}
 base_dyn_JacketLapels_r_01 {parent=clavicle_r, loc=<-5.1277 -1.7873 -5.4559>, rot=<0.0389 -0.999 0.0035 0.0221>}
 dyn_JacketLapels_r_01 {parent=base_dyn_JacketLapels_r_01, loc=<0 0 0>, rot=<0 0 0 1>}
 base_JacketLapels_r_02 {parent=dyn_JacketLapels_r_01, loc=<7.097 0 0>, rot=<-0.0149 -0.0901 0.0727 -0.9932>}
 dyn_JacketLapels_r_02 {parent=base_JacketLapels_r_02, loc=<0 0 0>, rot=<0 0 0 1>}
 attach_backpack {parent=spine_05, loc=<0.0161 -10.4703 -0.2056>, rot=<-0.0092 -0.7071 0.0096 -0.707>}
 Backpack_BR {parent=attach_backpack, type=Null, loc=<0 1.6 0>, rot=<-0.0698 0 0 0.9976>}
 attach_cape {parent=spine_05, loc=<9.8678 -13.4556 0.1282>, rot=<0.0617 -0.7045 -0.0613 -0.7044>}
 cape {parent=attach_cape, type=Null, loc=<0 0 0>, rot=<0 0 0 1>}
 base_dyn_backCollar {parent=spine_05, loc=<12.5472 -3.956 -0.0295>, rot=<0.6728 -0.2515 0.2661 -0.6429>}
 dyn_backCollar {parent=base_dyn_backCollar, loc=<0 0 0>, rot=<0 0 0 1>}
 base_dyn_JacketLapels_Mid_l_02 {parent=spine_05, loc=<3.0746 9.0064 -7.9329>, rot=<0.1364 -0.6468 -0.4066 -0.6306>}
 dyn_JacketLapels_Mid_l_02 {parent=base_dyn_JacketLapels_Mid_l_02, loc=<0 0 0>, rot=<0 0 0 1>}
 base_dyn_JacketLapels_Mid_r_02 {parent=spine_05, loc=<0.0914 8.8981 7.7656>, rot=<-0.4722 -0.5769 -0.4851 0.4571>}
 dyn_JacketLapels_Mid_r_02 {parent=base_dyn_JacketLapels_Mid_r_02, loc=<0 0 0>, rot=<0 0 0 1>}
 clavicle_l_null {parent=spine_05, type=Null, loc=<5.483 -0.1053 -3.2946>, rot=<0.0465 -0.7289 -0.009 -0.6829>}
 clavicle_r_null {parent=spine_05, type=Null, loc=<5.4831 -0.1053 3.2946>, rot=<0.7289 0.0465 -0.6829 0.009>}
 deform_pec_aim_base_l {parent=spine_04, loc=<3.2456 9.3309 -1.2237>, rot=<0.0603 -0.4033 0.3318 -0.8507>}
 deform_pec_aim_l {parent=deform_pec_aim_base_l, loc=<0 0 0>, rot=<0 0 0 1>}
 deform_pec_base_l {parent=deform_pec_aim_l, loc=<9.3665 0 0>, rot=<0 0 0 1>}
 deform_pec_l {parent=deform_pec_base_l, loc=<0 0 0>, rot=<0 0 0 1>}
 deform_pec_aim_base_r {parent=spine_04, loc=<3.2452 9.3309 1.2237>, rot=<-0.4033 -0.0603 0.8507 0.3318>}
 deform_pec_aim_r {parent=deform_pec_aim_base_r, loc=<0 0 0>, rot=<0 0 0 1>}
 deform_pec_base_r {parent=deform_pec_aim_r, loc=<-9.3667 -0.0003 -0.0004>, rot=<0 0 0 1>}
 deform_pec_r {parent=deform_pec_base_r, loc=<0 0 0>, rot=<0 0 0 1>}
 base_dyn_pin_03 {parent=spine_04, loc=<-2.4483 11.5466 -11.7536>, rot=<0.0268 -0.7063 -0.0259 -0.7069>}
 dyn_pin_03 {parent=base_dyn_pin_03, loc=<0 0 0>, rot=<0 0 0 1>}
 base_dyn_pin_02 {parent=spine_04, loc=<5.4714 13.5379 7.5888>, rot=<0.0066 -0.739 0.0122 -0.6735>}
 dyn_pin_02 {parent=base_dyn_pin_02, loc=<0 0 0>, rot=<0 0 0 1>}
 base_back_jacket_m {parent=spine_03, loc=<4.6864 -9.8055 -0.0381>, rot=<0.1021 -0.6837 0.7155 -0.101>}
 dyn_back_jacket_m {parent=base_back_jacket_m, loc=<0 0 0>, rot=<0 0 0 1>}
 base_side_jacket_l {parent=spine_03, loc=<1.6457 -1.5236 -14.1093>, rot=<0.0625 -0.9898 0.0899 -0.091>}
 dyn_side_jacket_l {parent=base_side_jacket_l, loc=<0 0 0>, rot=<0 0 0 1>}
 base_front_jacket_l {parent=spine_03, loc=<0.189 9.3081 -12.2831>, rot=<0.0224 -0.961 -0.2559 -0.1023>}
 dyn_front_jacket_l {parent=base_front_jacket_l, loc=<0 0 0>, rot=<0 0 0 1>}
 base_Zipper {parent=dyn_front_jacket_l, loc=<12.5146 4.9757 -5.4874>, rot=<-0.2788 0.0604 -0.0785 -0.9552>}
 dyn_Zipper {parent=base_Zipper, loc=<0 0 0>, rot=<0 0 0 1>}
 dyn_ZipperPin {parent=dyn_Zipper, loc=<1.7216 0.1215 0.0453>, rot=<0.0091 -0.0287 0.0347 -0.9989>}
 base_side_jacket_r {parent=spine_03, loc=<1.6456 -1.5236 14.1093>, rot=<0.1076 0.1255 0.9854 -0.0402>}
 dyn_side_jacket_r {parent=base_side_jacket_r, loc=<0 0 0>, rot=<0 0 0 1>}
 base_front_jacket_r {parent=spine_03, loc=<0.2571 8.8029 10.7675>, rot=<0.0858 -0.1299 0.9857 -0.0647>}
 dyn_front_jacket_r {parent=base_front_jacket_r, loc=<0 0 0>, rot=<0 0 0 1>}
 base_dyn_jacketRing {parent=dyn_front_jacket_r, loc=<11.5955 -2.7577 -7.2229>, rot=<-0.6542 0.176 0.6998 -0.2266>}
 dyn_jacketRing {parent=base_dyn_jacketRing, loc=<0 0 0>, rot=<0 0 0 1>}
 base_dyn_pin_01 {parent=spine_02, loc=<7.1514 8.6146 14.6808>, rot=<-0.0014 -0.7722 0.0058 -0.6354>}
 dyn_pin_01 {parent=base_dyn_pin_01, loc=<0 0 0>, rot=<0 0 0 1>}
 thigh_l {parent=pelvis, loc=<-1.0855 -0.2109 -8.7053>, rot=<0.0009 0.0355 0.0155 -0.9993>}
 calf_l {parent=thigh_l, loc=<-42.5284 -0.0019 -0.7297>, rot=<0.0006 -0.0161 -0.0391 -0.9991>}
 foot_l {parent=calf_l, loc=<-46.7401 -2.5014 -0.592>, rot=<0.0427 -0.0238 0.0353 -0.9982>}
 ball_l {parent=foot_l, loc=<-3.5658 10.8066 -0.1544>, rot=<0.0001 -0.0001 -0.7063 -0.7079>}
 calf_twist_02_l {parent=calf_l, loc=<-19.1546 -1.7943 -0.2638>, rot=<0.0237 0.0095 -0.0035 -0.9997>}
 calf_twist_01_l {parent=calf_l, loc=<-35.8483 -1.9197 -0.5794>, rot=<0.0237 0.0095 -0.0035 -0.9997>}
 deform_knee_in_base_l {parent=calf_l, loc=<0.8031 -5.0231 0.2456>, rot=<0 0 0 -1>}
 deform_knee_in_l {parent=deform_knee_in_base_l, loc=<0 0 0>, rot=<0 0 0 1>}
 deform_knee_out_base_l {parent=calf_l, loc=<0 4 0>, rot=<0 0 0 -1>}
 deform_knee_out_l {parent=deform_knee_out_base_l, loc=<0 0 0>, rot=<0 0 0 1>}
 thigh_twist_01_l {parent=thigh_l, loc=<-27.952 0 0>, rot=<0.0013 0 0 -1>}
 deform_groin_base_l {parent=thigh_l, loc=<0.4801 5.6776 -0.2486>, rot=<0 0 0 -1>}
 deform_groin_l {parent=deform_groin_base_l, loc=<0 0 0>, rot=<0 0 0 1>}
 thigh_r {parent=pelvis, loc=<-1.0855 -0.2109 8.7053>, rot=<0.0355 -0.0009 0.9993 0.0155>}
 calf_r {parent=thigh_r, loc=<42.5284 0.0019 0.7297>, rot=<0.0006 -0.0161 -0.0391 -0.9991>}
 foot_r {parent=calf_r, loc=<46.7401 2.5014 0.592>, rot=<0.0427 -0.0238 0.0353 -0.9982>}
 ball_r {parent=foot_r, loc=<3.5658 -10.8066 0.1544>, rot=<0.0001 -0.0001 -0.7063 -0.7079>}
 calf_twist_02_r {parent=calf_r, loc=<19.1546 1.7943 0.2638>, rot=<0.0237 0.0095 -0.0035 -0.9997>}
 calf_twist_01_r {parent=calf_r, loc=<35.8483 1.9197 0.5794>, rot=<0.0237 0.0095 -0.0035 -0.9997>}
 deform_knee_in_base_r {parent=calf_r, loc=<-0.8031 5.0231 -0.2457>, rot=<0 0 0 -1>}
 deform_knee_in_r {parent=deform_knee_in_base_r, loc=<0 0 0>, rot=<0 0 0 1>}
 deform_knee_out_base_r {parent=calf_r, loc=<0 -4 0>, rot=<0 0 0 -1>}
 deform_knee_out_r {parent=deform_knee_out_base_r, loc=<0 0 0>, rot=<0 0 0 1>}
 thigh_twist_01_r {parent=thigh_r, loc=<27.952 0 0>, rot=<0.0013 0 0 -1>}
 deform_groin_base_r {parent=thigh_r, loc=<-0.4801 -5.6776 0.2486>, rot=<0 0 0 -1>}
 deform_groin_r {parent=deform_groin_base_r, loc=<0 0 0>, rot=<0 0 0 1>}
 deform_glute_base_l {parent=pelvis, loc=<-7.3466 -9.0213 -9.1245>, rot=<0.0172 0.031 -0.4477 -0.8935>}
 deform_glute_l {parent=deform_glute_base_l, loc=<0 0 0>, rot=<0 0 0 1>}
 deform_glute_base_r {parent=pelvis, loc=<-7.3464 -9.0213 9.1245>, rot=<0.031 -0.0172 0.8935 -0.4477>}
 deform_glute_r {parent=deform_glute_base_r, loc=<0 0 0>, rot=<0 0 0 1>}
 base_center_belt {parent=pelvis, loc=<2.4198 9.3807 -0.1682>, rot=<-0.0127 -0.707 0.0127 -0.707>}
 dyn_center_belt {parent=base_center_belt, loc=<0 0 0>, rot=<0 0 0 1>}
 base_legbelt_l {parent=base_center_belt, loc=<12.3379 -2.0602 -7.883>, rot=<-0.1 0.1 -0.7 -0.7>}
 dyn_legbelt_l {parent=base_legbelt_l, loc=<0 0 0>, rot=<0 0 0 1>}
 base_back_belt {parent=pelvis, loc=<3.273 -11.6558 0.0017>, rot=<-0.0127 -0.707 0.0127 -0.707>}
 dyn_back_belt {parent=base_back_belt, loc=<0 0 0>, rot=<0 0 0 1>}
 clavicle_r_ctrl {controlType=EULER_TRANSFORM, type=Control, loc=<0 0 0>, rot=<0 0 0 1>}
 upperarm_r_ctrl {controlType=EULER_TRANSFORM, parent=clavicle_r_ctrl, type=Control, loc=<0 0 0>, rot=<0 0 0 1>}
 clavicle_l_ctrl {controlType=EULER_TRANSFORM, type=Control, loc=<0 0 0>, rot=<0 0 0 1>}
 upperarm_l_ctrl {controlType=EULER_TRANSFORM, parent=clavicle_l_ctrl, type=Control, loc=<0 0 0>, rot=<0 0 0 1>}
]
variables [
 IsEmoteActive {type=double, default=0}
 IshandglaidingActice {type=bool, default=True}
 isPistol {type=bool, default=False}
]
