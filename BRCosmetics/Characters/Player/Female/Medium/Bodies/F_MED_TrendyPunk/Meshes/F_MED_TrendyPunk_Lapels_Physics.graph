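# Physics asset for the jacket lapels and back collar of the F_MED_TrendyPunk female body cosmetic, setting up their secondary motion.
# Large capsules on spine, neck and upper arms form the body volumes around the cloth bones.
# Each dyn bone hangs with narrow angular limits from a tiny base body free-jointed to spine_05 or the preceding lapel bone.
# Left and right tuning differs, with the right mid lapel allowed far wider swing.
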PhysicsAsset /BRCosmetics/Characters/Player/Female/Medium/Bodies/F_MED_TrendyPunk/Meshes/F_MED_TrendyPunk_Lapels_Physics
bodies {
 base_JacketLapels_l_02 [{shapeType=Capsule, length=0.505, radius=0.505, shapeName=base_JacketLapels_l_02_capsule, rotation=<rot 0 -90 0>}]
 base_JacketLapels_r_02 [{shapeType=Capsule, length=0.505, radius=0.505, shapeName=base_JacketLapels_r_02_capsule, rotation=<rot 0 -90 0>}]
 base_dyn_JacketLapels_Mid_l_02 [{shapeType=Capsule, length=0.505, radius=0.505, shapeName=base_dyn_JacketLapels_Mid_l_02_capsule, rotation=<rot 0 -90 0>}]
 base_dyn_JacketLapels_Mid_r_02 [{shapeType=Capsule, length=0.505, radius=0.505, shapeName=base_dyn_JacketLapels_Mid_r_02_capsule, rotation=<rot 0 -90 0>}]
 base_dyn_JacketLapels_l_01 [{shapeType=Capsule, length=0.505, radius=0.505, shapeName=base_dyn_JacketLapels_l_01_capsule, rotation=<rot 0 -90 0>}]
 base_dyn_JacketLapels_r_01 [{shapeType=Capsule, length=0.505, radius=0.505, shapeName=base_dyn_JacketLapels_r_01_capsule, rotation=<rot 0 -90 0>}]
 base_dyn_backCollar [{shapeType=Capsule, length=0.505, radius=0.505, shapeName=base_dyn_backCollar_capsule, rotation=<rot 0 -90 0>}]
 dyn_JacketLapels_Mid_l_02 [{shapeType=Capsule, length=12, radius=1, center=<3.2059 0.3341 0.9179>, rotation=<rot 70.0001 90.0002 0.0001>}]
 dyn_JacketLapels_Mid_r_02 [{
   shapeType=Capsule
   length=16.9531
   radius=1.7028
   shapeName=dyn_JacketLapels_Mid_r_02_capsule
   center=<3.643 1.2412 2.389>
   rotation=<rot 55.4635 90 10.1414>
  }]
 dyn_JacketLapels_l_01 [{shapeType=Capsule, length=3, radius=2, shapeName=dyn_JacketLapels_l_01_capsule, center=<2.0091 0 2.1179>, rotation=<rot 90 0 0>}]
 dyn_JacketLapels_l_02 [{
   shapeType=Capsule
   length=3
   radius=2
   shapeName=dyn_JacketLapels_l_02_capsule
   center=<2.4557 0 3.5868>
   rotation=<rot 70 -179.9997 -179.9997>
  }]
 dyn_JacketLapels_r_01 [{
   shapeType=Capsule
   length=3
   radius=2
   shapeName=dyn_JacketLapels_r_01_capsule
   center=<3.4704 0 2.7401>
   rotation=<rot 0.0007 -95.0001 94.9999>
  }]
 dyn_JacketLapels_r_02 [{
   shapeType=Capsule
   length=3
   radius=2
   shapeName=dyn_JacketLapels_r_02_capsule
   center=<3.4704 0 1.9261>
   rotation=<rot 0.0005 -95.0002 89.9999>
  }]
 dyn_backCollar [{shapeType=Capsule, length=3, radius=2, shapeName=dyn_backCollar_capsule, center=<4.7878 0 0>, rotation=<rot 90 0 0>}]
 neck_01 [{shapeType=Capsule, length=3.4659, radius=7.5099, center=<-3.7714 1.4919 0>, rotation=<rot 90 0 -14.9998>}]
 spine_04 [{shapeType=Capsule, length=10, radius=10, shapeName=spine_04_capsule, center=<-0.602 0.1081 0>}]
 spine_05 [{shapeType=Capsule, length=19, radius=9, shapeName=spine_05_capsule, center=<0.3094 -0.0692 0>, rotation=<rot 0 -0.0001 0>}]
 upperarm_l [{
   shapeType=Capsule
   length=16
   radius=8.9202
   shapeName=upperarm_r_capsule
   center=<11.9674 0.3708 -0.9304>
   rotation=<rot 85 177.4754 0.0003>
  }]
 upperarm_r [{
   shapeType=Capsule
   length=16.1132
   radius=7.5096
   shapeName=upperarm_l_capsule
   center=<-10.6878 -0.6531 0.7426>
   rotation=<rot 86.3154 -178.6337 -179.9995>
  }]
}
constraints [
 {
  bone1Name=spine_05
  bone2Name=spine_04
  swing1LimitDegrees=45
  swing1Motion=Free
  swing2LimitDegrees=45
  swing2Motion=Free
  twistLimitDegrees=45
  twistMotion=Free
 }
 {
  bone1Name=upperarm_r
  bone2Name=spine_05
  swing1LimitDegrees=45
  swing1Motion=Free
  swing2LimitDegrees=45
  swing2Motion=Free
  twistLimitDegrees=45
  twistMotion=Free
 }
 {
  bone1Name=upperarm_l
  bone2Name=spine_05
  swing1LimitDegrees=45
  swing1Motion=Free
  swing2LimitDegrees=45
  swing2Motion=Free
  twistLimitDegrees=45
  twistMotion=Free
 }
 {
  bone1Name=base_dyn_JacketLapels_r_01
  bone2Name=spine_05
  swing1LimitDegrees=45
  swing1Motion=Free
  swing2LimitDegrees=45
  swing2Motion=Free
  twistLimitDegrees=45
  twistMotion=Free
 }
 {bone1Name=dyn_JacketLapels_r_01, bone2Name=base_dyn_JacketLapels_r_01, swing1LimitDegrees=5, swing2LimitDegrees=3, twistLimitDegrees=2}
 {
  bone1Name=base_JacketLapels_r_02
  bone2Name=dyn_JacketLapels_r_01
  swing1LimitDegrees=45
  swing1Motion=Free
  swing2LimitDegrees=45
  swing2Motion=Free
  twistLimitDegrees=45
  twistMotion=Free
 }
 {bone1Name=dyn_JacketLapels_r_02, bone2Name=base_JacketLapels_r_02, swing1LimitDegrees=5, swing2LimitDegrees=3, twistLimitDegrees=2}
 {
  bone1Name=base_dyn_JacketLapels_l_01
  bone2Name=spine_05
  swing1LimitDegrees=45
  swing1Motion=Free
  swing2LimitDegrees=45
  swing2Motion=Free
  twistLimitDegrees=45
  twistMotion=Free
 }
 {bone1Name=dyn_JacketLapels_l_01, bone2Name=base_dyn_JacketLapels_l_01, swing1LimitDegrees=2, swing2LimitDegrees=2, twistLimitDegrees=0}
 {
  bone1Name=base_JacketLapels_l_02
  bone2Name=dyn_JacketLapels_l_01
  swing1LimitDegrees=45
  swing1Motion=Free
  swing2LimitDegrees=45
  swing2Motion=Free
  twistLimitDegrees=45
  twistMotion=Free
 }
 {bone1Name=dyn_JacketLapels_l_02, bone2Name=base_JacketLapels_l_02, swing1LimitDegrees=2, swing2LimitDegrees=2, twistLimitDegrees=0}
 {
  bone1Name=base_dyn_backCollar
  bone2Name=spine_05
  swing1LimitDegrees=45
  swing1Motion=Free
  swing2LimitDegrees=45
  swing2Motion=Free
  twistLimitDegrees=45
  twistMotion=Free
 }
 {bone1Name=dyn_backCollar, bone2Name=base_dyn_backCollar, swing1LimitDegrees=5, swing2LimitDegrees=3, twistLimitDegrees=1}
 {
  bone1Name=neck_01
  bone2Name=spine_05
  swing1LimitDegrees=45
  swing1Motion=Free
  swing2LimitDegrees=45
  swing2Motion=Free
  twistLimitDegrees=45
  twistMotion=Free
 }
 {
  bone1Name=base_dyn_JacketLapels_Mid_l_02
  bone2Name=spine_05
  swing1LimitDegrees=45
  swing1Motion=Free
  swing2LimitDegrees=45
  swing2Motion=Free
  twistLimitDegrees=45
  twistMotion=Free
 }
 {
  bone1Name=dyn_JacketLapels_Mid_l_02
  bone2Name=base_dyn_JacketLapels_Mid_l_02
  swing1LimitDegrees=5
  swing2LimitDegrees=3
  twistLimitDegrees=2
 }
 {
  bone1Name=dyn_JacketLapels_Mid_r_02
  bone2Name=base_dyn_JacketLapels_Mid_r_02
  swing1LimitDegrees=10
  swing2LimitDegrees=140
  twistLimitDegrees=10
 }
 {
  bone1Name=base_dyn_JacketLapels_Mid_r_02
  bone2Name=spine_05
  swing1LimitDegrees=45
  swing1Motion=Free
  swing2LimitDegrees=45
  swing2Motion=Free
  twistLimitDegrees=45
  twistMotion=Free
 }
]
