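# Physics asset for the F_MED_TrendyPunk female medium body, set up for secondary motion of its jacket panels and right lapel.
# Capsules on spine, clavicles and thighs form the body volume, with free constraints linking spine and clavicles.
# Front, side and lapel dyn bones hinge from tiny base anchor bodies, while the back panel attaches directly to spine_03.
# Tight angular limits keep each panel swinging mostly on one axis.
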PhysicsAsset /BRCosmetics/Characters/Player/Female/Medium/Bodies/F_MED_TrendyPunk/Meshes/F_MED_TrendyPunk_Physics
bodies {
 base_dyn_JacketLapels_r_01 [{shapeType=Capsule, length=0.505, radius=0.505, shapeName=base_dyn_JacketLapels_r_01_capsule, rotation=<rot 0 -90 0>}]
 base_front_jacket_l [{shapeType=Capsule, length=0.505, radius=0.505, shapeName=base_front_jacket_l_capsule}]
 base_front_jacket_r [{shapeType=Capsule, length=0.505, radius=0.505, shapeName=base_front_jacket_r_capsule}]
 base_side_jacket_l [{shapeType=Capsule, length=0.505, radius=0.505, shapeName=base_side_jacket_l_capsule}]
 base_side_jacket_r [{shapeType=Capsule, length=0.505, radius=0.505, shapeName=base_side_jacket_r_capsule}]
 clavicle_l [{shapeType=Capsule, length=7, radius=6, shapeName=clavicle_l_capsule, center=<5.9572 -0.051 -2.1905>, rotation=<rot 90 0 0>}]
 clavicle_r [{shapeType=Capsule, length=7, radius=6, shapeName=clavicle_r_capsule, center=<-5.9572 0.051 2.1905>, rotation=<rot 90 0 180>}]
 dyn_JacketLapels_r_01 [{shapeType=Capsule, length=6, radius=3, shapeName=dyn_JacketLapels_r_01_capsule, center=<7.9741 -2.6002 0.9711>, rotation=<rot 90 0 -69.9999>}]
 dyn_back_jacket_m [{shapeType=Capsule, length=3, radius=2, center=<1 0 0>, rotation=<rot 90 0 0>}]
 dyn_front_jacket_l [{shapeType=Capsule, length=3, radius=2, center=<5 0 0>, rotation=<rot 90 0 0>}]
 dyn_front_jacket_r [{shapeType=Capsule, length=3, radius=2, center=<1 0 0>, rotation=<rot 90 0 0>}]
 dyn_side_jacket_l [{shapeType=Capsule, length=3, radius=2, center=<1 0 0>, rotation=<rot 90 0 0>}]
 dyn_side_jacket_r [{shapeType=Capsule, length=3, radius=2, center=<1 0 0>, rotation=<rot 90 0 0>}]
 spine_01 [{shapeType=Capsule, length=10.9402, radius=8.743, shapeName=spine_01_capsule, center=<-3.4407 0.5783 0>}]
 spine_02 [{shapeType=Capsule, length=6, radius=8, center=<-2.8093 0.4764 0>}]
 spine_03 [{shapeType=Capsule, length=6, radius=8}]
 spine_05 [{shapeType=Capsule, length=8, radius=8, shapeName=spine_05_capsule, center=<1.5914 0.9551 0>}]
 thigh_l [{shapeType=Capsule, length=12, radius=8, center=<-11.3963 -1.6513 -0.0421>, rotation=<rot 90 3.5763 13.5764>}]
 thigh_r [{shapeType=Capsule, length=11.0075, radius=8.4539, center=<9.4533 -2.0527 0>, rotation=<rot 84.5883 169.7933 -0.0002>}]
}
constraints [
 {
  bone1Name=spine_03
  bone2Name=spine_02
  swing1LimitDegrees=45
  swing1Motion=Free
  swing2LimitDegrees=45
  swing2Motion=Free
  twistLimitDegrees=45
  twistMotion=Free
 }
 {
  bone1Name=base_front_jacket_l
  bone2Name=spine_03
  swing1LimitDegrees=45
  swing1Motion=Free
  swing2LimitDegrees=45
  swing2Motion=Free
  twistLimitDegrees=45
  twistMotion=Free
 }
 {
  bone1Name=base_front_jacket_r
  bone2Name=spine_03
  swing1LimitDegrees=45
  swing1Motion=Free
  swing2LimitDegrees=45
  swing2Motion=Free
  twistLimitDegrees=45
  twistMotion=Free
 }
 {
  bone1Name=base_side_jacket_l
  bone2Name=spine_03
  swing1LimitDegrees=45
  swing1Motion=Free
  swing2LimitDegrees=45
  swing2Motion=Free
  twistLimitDegrees=45
  twistMotion=Free
 }
 {
  bone1Name=base_side_jacket_r
  bone2Name=spine_03
  swing1LimitDegrees=45
  swing1Motion=Free
  swing2LimitDegrees=45
  swing2Motion=Free
  twistLimitDegrees=45
  twistMotion=Free
 }
 {bone1Name=dyn_back_jacket_m, bone2Name=spine_03, swing1LimitDegrees=0, swing2LimitDegrees=20, twistLimitDegrees=0}
 {bone1Name=dyn_front_jacket_l, bone2Name=base_front_jacket_l, swing1LimitDegrees=10, swing2LimitDegrees=0, twistLimitDegrees=0}
 {bone1Name=dyn_front_jacket_r, bone2Name=base_front_jacket_r, swing1LimitDegrees=0, swing2LimitDegrees=15, twistLimitDegrees=0}
 {bone1Name=dyn_side_jacket_l, bone2Name=base_side_jacket_l, swing1LimitDegrees=0, swing2LimitDegrees=15, twistLimitDegrees=0}
 {bone1Name=dyn_side_jacket_r, bone2Name=base_side_jacket_r, swing1LimitDegrees=0, swing2LimitDegrees=15, twistLimitDegrees=0}
 {
  bone1Name=spine_05
  bone2Name=spine_03
  swing1LimitDegrees=45
  swing1Motion=Free
  swing2LimitDegrees=45
  swing2Motion=Free
  twistLimitDegrees=45
  twistMotion=Free
 }
 {
  bone1Name=clavicle_l
  bone2Name=spine_05
  swing1LimitDegrees=45
  swing1Motion=Free
  swing2LimitDegrees=45
  swing2Motion=Free
  twistLimitDegrees=45
  twistMotion=Free
 }
 {
  bone1Name=clavicle_r
  bone2Name=spine_05
  swing1LimitDegrees=45
  swing1Motion=Free
  swing2LimitDegrees=45
  swing2Motion=Free
  twistLimitDegrees=45
  twistMotion=Free
 }
 {
  bone1Name=base_dyn_JacketLapels_r_01
  bone2Name=spine_05
  swing1LimitDegrees=45
  swing1Motion=Free
  swing2LimitDegrees=45
  swing2Motion=Free
  twistLimitDegrees=45
  twistMotion=Free
 }
 {bone1Name=dyn_JacketLapels_r_01, bone2Name=base_dyn_JacketLapels_r_01, swing1LimitDegrees=3, swing2LimitDegrees=3, twistLimitDegrees=0}
 {
  bone1Name=spine_02
  bone2Name=spine_01
  swing1LimitDegrees=45
  swing1Motion=Free
  swing2LimitDegrees=45
  swing2Motion=Free
  twistLimitDegrees=45
  twistMotion=Free
 }
]
