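# Animation Blueprint for the F_MED_TrendyPunk female cosmetic body, adding deformation and secondary motion to the jacket, zipper and belt.
# A Control Rig processes the parent mesh's copied pose, taking hang-gliding, pistol and Mime-emote states set in the EventGraph.
# Two RigidBody passes, lapel offsets and spring AnimDynamics on sleeves, ring and zipper follow, with a weighted constraint placing the zipper base.
# Final additive bone offsets adjust jacket fronts and belt.
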AnimBlueprint /BRCosmetics/Characters/Player/Female/Medium/Bodies/F_MED_TrendyPunk/Meshes/F_MED_TrendyPunk_AnimBP
target_skeleton=&/F_MED_TrendyPunk_Skeleton
graphs [
 AnimGraph {nodes [
   Output Pose(Result=Use cached pose 'CRCache_Adj'()) {layerGroup=None}
   n1 = Control Rig!(
    Source=Copy Pose From Mesh() {bUseAttachedParent=true}
    Alpha
    IshandglaidingActice:bool=Get isHangGliding() -> (isHangGliding:bool)
    isPistol:bool=Get isPistol() -> (isPistol:bool)
    IsEmoteActive:double=Get isCustomEmotePlaying_c() -> (isCustomEmotePlaying_c:double)
   ) -> (Pose) {
    alphaBoolBlend {blendOption=Linear}
    controlRigAssetReference {blueprintRigClass=&/F_MED_TrendyPunk_Deform_CtrlRig.F_MED_TrendyPunk_Deform_CtrlRig_C}
   }
   Save cached pose 'CRCache_Input'(Pose=@n1) {cacheName=CRCache_Input}
   n2 = RigidBody!(ComponentPose=Local To Component(LocalPose=Use cached pose 'CRCache_Input'()), Alpha, OverrideWorldGravity:Vector="0,0,0", ExternalForce) -> (Pose) {
    bOverrideWorldGravity=true
    overridePhysicsAsset=&/F_MED_TrendyPunk_Physics
    simulationSpace=BaseBoneSpace
    alphaBoolBlend {blendOption=Linear}
    baseBoneRef {boneName=pelvis}
    simSpaceSettings {
     dampingAlpha=0
     maxAngularAcceleration=40
     maxAngularVelocity=10
     maxLinearAcceleration=4000
     maxLinearVelocity=800
     velocityScaleZ=0.6
     worldAlpha=1
     externalLinearDragV=<1.5 1.5 2>
    }
   }
   n3 = RigidBody!(ComponentPose=@n2, Alpha, OverrideWorldGravity="0,0,0") -> (Pose) {
    bOverrideWorldGravity=true
    overridePhysicsAsset=&/F_MED_TrendyPunk_Lapels_Physics
    simulationSpace=BaseBoneSpace
    alphaBoolBlend {blendOption=Linear}
    alphaScaleBiasClamp {interpSpeedDecreasing=0.1, interpSpeedIncreasing=0.1}
    baseBoneRef {boneName=spine_05}
    componentAppliedLinearAccClamp=<3000 3000 3000>
    componentLinearAccScale=<0.01 0.01 0.01>
    componentLinearVelScale=<0.01 0.01 0.01>
    overrideWorldGravity {z=980}
    simSpaceSettings {maxAngularAcceleration=1000, maxAngularVelocity=3, maxLinearAcceleration=700, maxLinearVelocity=500, externalLinearDragV {x=0.2}}
   }
   n4 = "Transform (Modify) Bone - Bone: base_dyn_JacketLapels_r_01"(ComponentPose=@n3, Alpha=1, Translation="0,0,0", Rotation="20.000,-10.000,0.000", Scale="1,1,1") {
    rotationMode=BMM_Additive
    rotationSpace=BCS_BoneSpace
    translationMode=BMM_Additive
    translationSpace=BCS_BoneSpace
    alphaBoolBlend {blendOption=Linear}
    boneToModify {boneName=base_dyn_JacketLapels_r_01}
    rotation=<rot 20 -9.9998 0.0001>
   }
   n5 = "Transform (Modify) Bone - Bone: base_JacketLapels_r_02"(ComponentPose=@n4, Alpha=1, Translation="0,0,0", Rotation="14.998,-0.000,0.000", Scale="1,1,1") {
    rotationMode=BMM_Additive
    rotationSpace=BCS_BoneSpace
    translationMode=BMM_Additive
    translationSpace=BCS_BoneSpace
    alphaBoolBlend {blendOption=Linear}
    boneToModify {boneName=base_JacketLapels_r_02}
    rotation=<rot 14.998 0 0.0001>
   }
   n6 = "Transform (Modify) Bone - Bone: base_dyn_JacketLapels_l_01"(ComponentPose=@n5, Alpha=1, Translation="0,0,0", Rotation="15.000,-20.000,-0.000", Scale="1,1,1") {
    rotationMode=BMM_Additive
    rotationSpace=BCS_BoneSpace
    translationMode=BMM_Additive
    translationSpace=BCS_BoneSpace
    alphaBoolBlend {blendOption=Linear}
    boneToModify {boneName=base_dyn_JacketLapels_l_01}
    rotation=<rot 15 -20.0001 -0.0001>
   }
   n7 = Anim Dynamics - Bone: dyn_JacketSleeve_a_r!(ComponentPose=@n6, Alpha=1, GravityOverride:Vector="0,0,-980", ExternalForce="0,0,0") -> (Pose) {
    angularDampingOverride=1
    angularSpringConstant=5000
    bAngularSpring=true
    bOverrideAngularDamping=true
    bUseGravityOverride=true
    gravityScale=0.35
    alphaBoolBlend {blendOption=Linear}
    boundBone {boneName=dyn_JacketSleeve_a_r}
    componentAppliedLinearAccClamp=<200 200 200>
    componentLinearAccScale=<0.6 0.6 0.6>
    componentLinearVelScale=<0.01 0.01 0.01>
    gravityOverride {z=-980}
    physicsBodyDefinitions [{
      boundBone {boneName=dyn_JacketSleeve_a_r}
      boxExtents=<12 12 12>
      constraintSetup {coneAngle=20, angularLimitsMax=<5 5 0>, angularLimitsMin=<-5 -5 -7>, angularTarget=<1 0 0>}
      localJointOffset=<1 0 0>
     }]
   }
   n8 = Anim Dynamics - Bone: dyn_JacketSleeve_b_r!(ComponentPose=@n7, Alpha=1, GravityOverride:Vector="0,0,-980", ExternalForce="0,0,0") -> (Pose) {
    angularDampingOverride=1
    angularSpringConstant=5000
    bAngularSpring=true
    bOverrideAngularDamping=true
    bUseGravityOverride=true
    gravityScale=0.35
    alphaBoolBlend {blendOption=Linear}
    boundBone {boneName=dyn_JacketSleeve_b_r}
    componentAppliedLinearAccClamp=<200 200 200>
    componentLinearAccScale=<0.6 0.6 0.6>
    componentLinearVelScale=<0.01 0.01 0.01>
    gravityOverride {z=-980}
    physicsBodyDefinitions [{
      boundBone {boneName=dyn_JacketSleeve_b_r}
      boxExtents=<12 12 12>
      constraintSetup {coneAngle=20, angularLimitsMax=<5 0 5>, angularLimitsMin=<-5 -10 -5>, angularTarget=<1 0 0>}
      localJointOffset=<1 0 0>
     }]
   }
   n9 = Anim Dynamics - Bone: dyn_JacketSleeve_c_r!(ComponentPose=@n8, Alpha=1, GravityOverride:Vector="0,0,-980", ExternalForce="0,0,0") -> (Pose) {
    angularDampingOverride=1
    angularSpringConstant=5000
    bAngularSpring=true
    bOverrideAngularDamping=true
    bUseGravityOverride=true
    gravityScale=0.35
    alphaBoolBlend {blendOption=Linear}
    boundBone {boneName=dyn_JacketSleeve_c_r}
    componentAppliedLinearAccClamp=<200 200 200>
    componentLinearAccScale=<0.6 0.6 0.6>
    componentLinearVelScale=<0.01 0.01 0.01>
    gravityOverride {z=-980}
    physicsBodyDefinitions [{
      boundBone {boneName=dyn_JacketSleeve_c_r}
      boxExtents=<12 12 12>
      constraintSetup {coneAngle=20, angularLimitsMax=<5 20 5>, angularLimitsMin=<-5 0 -5>, angularTarget=<1 0 0>}
      localJointOffset=<1 0 0>
     }]
   }
   n10 = Anim Dynamics - Bone: dyn_JacketSleeve_d_r!(ComponentPose=@n9, Alpha=1, GravityOverride:Vector="0,0,-980", ExternalForce="0,0,0") -> (Pose) {
    angularDampingOverride=1
    angularSpringConstant=5000
    bAngularSpring=true
    bOverrideAngularDamping=true
    bUseGravityOverride=true
    gravityScale=0.35
    alphaBoolBlend {blendOption=Linear}
    boundBone {boneName=dyn_JacketSleeve_d_r}
    componentAppliedLinearAccClamp=<200 200 200>
    componentLinearAccScale=<0.6 0.6 0.6>
    componentLinearVelScale=<0.01 0.01 0.01>
    gravityOverride {z=-980}
    physicsBodyDefinitions [{
      boundBone {boneName=dyn_JacketSleeve_d_r}
      boxExtents=<12 12 12>
      constraintSetup {coneAngle=20, angularLimitsMax=<5 5 5>, angularLimitsMin=<-5 -5 -20>, angularTarget=<1 0 0>}
      localJointOffset=<1 0 0>
     }]
   }
   n11 = Anim Dynamics - Bone: dyn_JacketSleeve_a_l!(ComponentPose=@n10, Alpha=1, GravityOverride:Vector="0,0,-980", ExternalForce="0,0,0") -> (Pose) {
    angularDampingOverride=1
    angularSpringConstant=5000
    bAngularSpring=true
    bOverrideAngularDamping=true
    bUseGravityOverride=true
    gravityScale=0.35
    alphaBoolBlend {blendOption=Linear}
    boundBone {boneName=dyn_JacketSleeve_a_l}
    componentAppliedLinearAccClamp=<200 200 200>
    componentLinearAccScale=<0.6 0.6 0.6>
    componentLinearVelScale=<0.01 0.01 0.01>
    gravityOverride {z=-980}
    physicsBodyDefinitions [{
      boundBone {boneName=dyn_JacketSleeve_a_l}
      boxExtents=<12 12 12>
      constraintSetup {angularLimitsMax=<5 5 0>, angularLimitsMin=<-5 -5 -7>, angularTarget=<1 0 0>}
      localJointOffset=<1 0 0>
     }]
   }
   n12 = Anim Dynamics - Bone: dyn_JacketSleeve_b_l!(ComponentPose=@n11, Alpha=1, GravityOverride:Vector="0,0,-980", ExternalForce="0,0,0") -> (Pose) {
    angularDampingOverride=1
    angularSpringConstant=5000
    bAngularSpring=true
    bOverrideAngularDamping=true
    bUseGravityOverride=true
    gravityScale=0.35
    alphaBoolBlend {blendOption=Linear}
    boundBone {boneName=dyn_JacketSleeve_b_l}
    componentAppliedLinearAccClamp=<200 200 200>
    componentLinearAccScale=<0.6 0.6 0.6>
    componentLinearVelScale=<0.01 0.01 0.01>
    gravityOverride {z=-980}
    physicsBodyDefinitions [{
      boundBone {boneName=dyn_JacketSleeve_b_l}
      boxExtents=<12 12 12>
      constraintSetup {angularLimitsMax=<5 0 5>, angularLimitsMin=<-5 -10 -5>, angularTarget=<1 0 0>}
      localJointOffset=<1 0 0>
     }]
   }
   n13 = Anim Dynamics - Bone: dyn_JacketSleeve_c_l!(ComponentPose=@n12, Alpha=1, GravityOverride:Vector="0,0,-980", ExternalForce="0,0,0") -> (Pose) {
    angularDampingOverride=1
    angularSpringConstant=5000
    bAngularSpring=true
    bOverrideAngularDamping=true
    bUseGravityOverride=true
    gravityScale=0.35
    alphaBoolBlend {blendOption=Linear}
    boundBone {boneName=dyn_JacketSleeve_c_l}
    componentAppliedLinearAccClamp=<200 200 200>
    componentLinearAccScale=<0.6 0.6 0.6>
    componentLinearVelScale=<0.01 0.01 0.01>
    gravityOverride {z=-980}
    physicsBodyDefinitions [{
      boundBone {boneName=dyn_JacketSleeve_c_l}
      boxExtents=<12 12 12>
      constraintSetup {angularLimitsMax=<5 20 5>, angularLimitsMin=<-5 0 -5>, angularTarget=<1 0 0>}
      localJointOffset=<1 0 0>
     }]
   }
   n14 = Anim Dynamics - Bone: dyn_JacketSleeve_d_l!(ComponentPose=@n13, Alpha=1, GravityOverride:Vector="0,0,-980", ExternalForce="0,0,0") -> (Pose) {
    angularDampingOverride=1
    angularSpringConstant=5000
    bAngularSpring=true
    bOverrideAngularDamping=true
    bUseGravityOverride=true
    gravityScale=0.35
    alphaBoolBlend {blendOption=Linear}
    boundBone {boneName=dyn_JacketSleeve_d_l}
    componentAppliedLinearAccClamp=<200 200 200>
    componentLinearAccScale=<0.6 0.6 0.6>
    componentLinearVelScale=<0.01 0.01 0.01>
    gravityOverride {z=-980}
    physicsBodyDefinitions [{
      boundBone {boneName=dyn_JacketSleeve_d_l}
      boxExtents=<12 12 12>
      constraintSetup {angularLimitsMax=<5 5 5>, angularLimitsMin=<-5 -5 -20>, angularTarget=<1 0 0>}
      localJointOffset=<1 0 0>
     }]
   }
   n15 = Anim Dynamics - Bone: dyn_jacketRing!(ComponentPose=@n14, Alpha=1, GravityOverride:Vector="0,0,-980", ExternalForce="0,0,0") -> (Pose) {
    angularDampingOverride=1
    angularSpringConstant=6000
    bAngularSpring=true
    bOverrideAngularDamping=true
    bUseGravityOverride=true
    gravityScale=0.35
    simulationSpace=RootRelative
    alphaBoolBlend {blendOption=Linear}
    boundBone {boneName=dyn_jacketRing}
    componentLinearAccScale=<1 1 1>
    componentLinearVelScale=<1 1 1>
    gravityOverride {z=-980}
    physicsBodyDefinitions [{
      boundBone {boneName=dyn_jacketRing}
      boxExtents=<12 15 12>
      constraintSetup {angularLimitsMax=<30 30 30>, angularLimitsMin=<-30 -30 -30>, angularTarget=<1 0 0>}
      localJointOffset=<1 0 0>
     }]
   }
   n16 = "Constraint - base_Zipper (PPPPPPP)"!(
    ComponentPose=@n15
    Alpha=1
    ConstraintWeights_0=0.9284
    ConstraintWeights_1:float=0.0683
    ConstraintWeights_2:float=0.0012
    ConstraintWeights_3:float=0.0012
    ConstraintWeights_4:float=0.0007
    ConstraintWeights_5:float=0.0002
    ConstraintWeights_6:float=0.0002
   ) -> (Pose) {
    alphaBoolBlend {blendOption=Linear}
    boneToModify {boneName=base_Zipper}
    constraintSetup [
     {targetBone {boneName=dyn_front_jacket_l}}
     {targetBone {boneName=dyn_side_jacket_l}}
     {targetBone {boneName=spine_02}}
     {targetBone {boneName=spine_04}}
     {targetBone {boneName=dyn_back_jacket_m}}
     {targetBone {boneName=pelvis}}
     {targetBone {boneName=spine_03}}
    ]
    constraintWeights [1,1,1,1,1,1,1]
   }
   n17 = Anim Dynamics - Bone: dyn_Zipper(ComponentPose=@n16, Alpha=1, ExternalForce="0,0,0") {
    angularDampingOverride=1
    angularSpringConstant=9000
    bAngularSpring=true
    bOverrideAngularDamping=true
    bUseGravityOverride=true
    gravityScale=0.35
    simulationSpace=RootRelative
    alphaBoolBlend {blendOption=Linear}
    boundBone {boneName=dyn_Zipper}
    componentAppliedLinearAccClamp=<10000 10000 10000>
    componentLinearAccScale=<1 1 1>
    componentLinearVelScale=<1 1 1>
    gravityOverride {z=-980}
    physicsBodyDefinitions [{
      boundBone {boneName=dyn_Zipper}
      boxExtents=<20 12 20>
      constraintSetup {angularLimitsMax=<5 5 5>, angularLimitsMin=<-5 -5 -5>, angularTarget=<1 0 0>}
      localJointOffset=<3 0 0>
     }]
   }
   n18 = Anim Dynamics - Bone: dyn_ZipperPin(ComponentPose=@n17, Alpha=0.4, ExternalForce="0,0,0") {
    angularDampingOverride=1
    angularSpringConstant=12000
    bAngularSpring=true
    bOverrideAngularDamping=true
    bUseGravityOverride=true
    gravityScale=0.35
    simulationSpace=RootRelative
    alphaBoolBlend {blendOption=Linear}
    boundBone {boneName=dyn_ZipperPin}
    componentAppliedLinearAccClamp=<10000 10000 10000>
    componentLinearAccScale=<1 1 1>
    componentLinearVelScale=<1 1 1>
    gravityOverride {z=-980}
    physicsBodyDefinitions [{
      boundBone {boneName=dyn_ZipperPin}
      boxExtents=<12 15 12>
      constraintSetup {angularLimitsMax=<30 30 30>, angularLimitsMin=<-30 -30 -30>, angularTarget=<1 0 0>}
      localJointOffset=<1 0 0>
     }]
   }
   Save cached pose 'CRCache_Dyn'(Pose=Component To Local(ComponentPose=@n18)) {cacheName=CRCache_Dyn}
   n19 = "Transform (Modify) Bone - Bone: base_front_jacket_r"!(
    ComponentPose=Local To Component(LocalPose=Use cached pose 'CRCache_Dyn'())
    bAlphaBoolEnabled:bool=True
    Translation="0,0,0"
    Rotation="-0.000,-38.505,0.000"
    Scale="1,1,1"
   ) -> (Pose) {
    alphaInputType=Bool
    rotationMode=BMM_Additive
    rotationSpace=BCS_BoneSpace
    translationMode=BMM_Additive
    translationSpace=BCS_BoneSpace
    alphaBoolBlend {blendInTime=0.25, blendOption=Linear, blendOutTime=0.25}
    boneToModify {boneName=base_front_jacket_r}
    rotation=<rot 0 -38.505 0>
   }
   n20 = "Transform (Modify) Bone - Bone: base_dyn_jacketRing"!(ComponentPose=@n19, bAlphaBoolEnabled:bool=True, Translation="0,0,0", Rotation="0,0,0", Scale="1,1,1") -> (Pose) {
    alphaInputType=Bool
    rotationMode=BMM_Additive
    rotationSpace=BCS_BoneSpace
    translationMode=BMM_Additive
    translationSpace=BCS_BoneSpace
    alphaBoolBlend {blendInTime=0.25, blendOption=Linear, blendOutTime=0.25}
    boneToModify {boneName=base_dyn_jacketRing}
   }
   n21 = "Transform (Modify) Bone - Bone: base_front_jacket_l"!(ComponentPose=@n20, bAlphaBoolEnabled:bool=True, Translation="0,0,0", Rotation="46.115,51.112,41.784", Scale="1,1,1") -> (Pose) {
    alphaInputType=Bool
    rotationMode=BMM_Additive
    rotationSpace=BCS_BoneSpace
    translationMode=BMM_Additive
    translationSpace=BCS_BoneSpace
    alphaBoolBlend {blendInTime=0.25, blendOption=Linear, blendOutTime=0.25}
    boneToModify {boneName=base_front_jacket_l}
    rotation=<rot 46.115 51.112 41.784>
   }
   n22 = "Transform (Modify) Bone - Bone: base_center_belt"(ComponentPose=@n21, Alpha=1, Translation="1.136,0.462,0.061", Rotation="-0.012,-0.011,-16.851", Scale="1,1,1") {
    rotationMode=BMM_Additive
    rotationSpace=BCS_BoneSpace
    scaleMode=BMM_Additive
    scaleSpace=BCS_BoneSpace
    translationMode=BMM_Additive
    translationSpace=BCS_BoneSpace
    alphaBoolBlend {blendOption=Linear}
    boneToModify {boneName=base_center_belt}
    rotation=<rot -0.012 -0.011 -16.8511>
    translation=<1.136 0.462 0.061>
   }
   n23 = "Transform (Modify) Bone - Bone: base_front_jacket_l"(ComponentPose=@n22, bAlphaBoolEnabled=True, Translation="-5.637,-1.435,-2.450", Rotation="36.742,19.215,7.268", Scale="1,1,1") {
    alphaInputType=Bool
    rotationMode=BMM_Additive
    rotationSpace=BCS_BoneSpace
    translationMode=BMM_Additive
    translationSpace=BCS_BoneSpace
    alphaBoolBlend {blendInTime=0.25, blendOption=Linear, blendOutTime=0.25}
    boneToModify {boneName=base_front_jacket_l}
    rotation=<rot 36.7421 19.2151 7.268>
    translation=<-5.637 -1.435 -2.45>
   }
   n24 = "Transform (Modify) Bone - Bone: dyn_center_belt"!(ComponentPose=@n23, bAlphaBoolEnabled:bool=True, Translation="0.278,0.737,2.662", Rotation="0.003,-0.001,-9.800", Scale="1,1,1") -> (Pose) {
    alphaInputType=Bool
    rotationMode=BMM_Additive
    rotationSpace=BCS_BoneSpace
    translationMode=BMM_Additive
    translationSpace=BCS_BoneSpace
    alphaBoolBlend {blendInTime=0.25, blendOption=Linear, blendOutTime=0.25}
    boneToModify {boneName=dyn_center_belt}
    rotation=<rot 0.003 -0.001 -9.8001>
    translation=<0.278 0.737 2.662>
   }
   Save cached pose 'CRCache_Adj'(Pose=Component To Local(ComponentPose=@n24)) {cacheName=CRCache_Adj}
  ]}
 EventGraph {nodes [
   Event BlueprintUpdateAnimation()
   > n1 = Sequence()
   > n2 = Gravity Override State Function DEPRECATED(CharacterPartAnimInstance=Self-Reference(), GameTimeInSeconds=GetGameTimeInSeconds(), GravityOverrideParams=Get Gravity Override Params())
   > Set GravityOverride(GravityOverride=@n2)
   n3 = GetAnimInstance(self=GetOwnerSkeletalMeshComponent())
   Is Valid(exec=@n1.then_1, InputObject=@n3)
   > Set isCustomEmotePlaying_b(isCustomEmotePlaying_b:double=SelectFloat(A=1, B=0, bPickA=Montage_IsPlaying(self=@n3, Montage=/BRCosmetics/Animation/Game/MainPlayer/Emotes/Emote_Celebration_Female_M))) -> (Output_Get:double)
   > Set isCustomEmotePlaying_c(isCustomEmotePlaying_c:double=SelectFloat(A=1, B=0, bPickA=Montage_IsPlaying(self=@n3, Montage=/BRCosmetics/Animation/Game/MainPlayer/Emotes/Mime/Emote_Mime_CMF_M))) -> (Output_Get:double)
   Event OnParachuteOpened()
   > Switch on EGliderType(Selection=Get PlayerGliderType() -> (PlayerGliderType=HangGlider))
   > Set isHangGliding(isHangGliding:bool=true) -> (Output_Get:bool)
   Event OnWeaponChanged()
   > Set isPistol(isPistol=false)
   > n4 = Switch on EFortWeaponCoreAnimation!(execute, Selection:EFortWeaponCoreAnimation Enum=Get WeaponCoreAnim!() -> (WeaponCoreAnim)) -> (
    Melee:Exec
    Pistol:Exec
    Shotgun:Exec
    PaperBlueprint:Exec
    Rifle:Exec
    MeleeOneHand:Exec
    MachinePistol:Exec
    RocketLauncher:Exec
    GrenadeLauncher:Exec
    GoingCommando:Exec
    AssaultRifle:Exec
    TacticalShotgun:Exec
    SniperRifle:Exec
    TrapPlacement:Exec
    ShoulderLauncher:Exec
    AbilityDecoTool:Exec
    Crossbow:Exec
    C4:Exec
    RemoteControl:Exec
    DualWield:Exec
    AR_BullPup:Exec
    AR_ForwardGrip:Exec
    MedPackPaddles:Exec
    SMG_P90:Exec
    AR_DrumGun:Exec
    Consumable_Small:Exec
    Consumable_Large:Exec
    Balloon:Exec
    MountedTurret:Exec
    CreativeTool:Exec
    ExplosiveBow:Exec
    AshtonIndigo:Exec
    AshtonChicago:Exec
    MeleeDualWield:Exec
    Unarmed:Exec
   )
   Set isPistol(execute=@n4.Pistol, isPistol=true)
   Event OnParachuteClosed()
   > Set isHangGliding(isHangGliding=false)
  ]}
]
variables [
 Gravity Override Params {type=GravityOverrideParamsStruct}
 GravityOverride {type=Vector}
 isCustomEmotePlaying {type=number}
 isCustomEmotePlaying_b {type=number}
 isCustomEmotePlaying_c {type=number}
 isHangGliding {type=boolean}
 isPistol {type=boolean}
 SkelMesh {type=SkeletalMeshComponent}
]
