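# Anim Blueprint for the TrendyPunk female character's face accessory part, adding secondary physics motion to beanie and hair bones.
# It copies the attached parent's pose, runs beanie AnimDynamics into a cached pose, then chains front, side and back hair AnimDynamics.
# Springs, angular limits and per-bone alphas restrain the motion.
# The EventGraph updates gravity-override and front-end-dependent simulation-space variables, and caches the body mesh on customization load.
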AnimBlueprint /BRCosmetics/Characters/Player/Female/Medium/Bodies/F_MED_TrendyPunk/Meshes/Parts/F_MED_TrendyPunk_FaceAcc_AnimBP
target_skeleton=&/F_MED_TrendyPunk_FaceAcc_Skeleton
graphs [
 AnimGraph {nodes [
   n1 = Anim Dynamics - Bone: dyn_frontHair_01!(
    SimulationSpace:AnimPhysSimSpaceType Enum
    ComponentPose=Local To Component(LocalPose=Use cached pose 'CachePhysics'())
    Alpha=0.6
    ExternalForce="0,0,0"
   ) -> (Pose) {
    angularDampingOverride=1
    angularSpringConstant=6000
    bAngularSpring=true
    bOverrideAngularDamping=true
    bUseGravityOverride=true
    gravityScale=0.35
    simulationSpace=RootRelative
    alphaBoolBlend {blendOption=Linear}
    boundBone {boneName=dyn_frontHair_01}
    componentAppliedLinearAccClamp=<10000 10000 10000>
    componentLinearAccScale=<1 1 1>
    componentLinearVelScale=<1 1 1>
    gravityOverride {z=-980}
    physicsBodyDefinitions [{
      boundBone {boneName=dyn_frontHair_01}
      boxExtents=<15 20 20>
      constraintSetup {angularLimitsMax=<10 10 10>, angularLimitsMin=<-5 -5 -5>, angularTarget=<1 0 0>}
      localJointOffset=<0.2 0 0>
     }]
   }
   n2 = Anim Dynamics - Bone: dyn_frontHair_02!(SimulationSpace:AnimPhysSimSpaceType Enum, ComponentPose=@n1, Alpha=0.4, ExternalForce="0,0,0") -> (Pose) {
    angularDampingOverride=1
    angularSpringConstant=6000
    bAngularSpring=true
    bOverrideAngularDamping=true
    bUseGravityOverride=true
    gravityScale=0.35
    simulationSpace=RootRelative
    alphaBoolBlend {blendOption=Linear}
    boundBone {boneName=dyn_frontHair_02}
    componentAppliedLinearAccClamp=<10000 10000 10000>
    componentLinearAccScale=<1 1 1>
    componentLinearVelScale=<1 1 1>
    gravityOverride {z=-980}
    physicsBodyDefinitions [{
      boundBone {boneName=dyn_frontHair_02}
      boxExtents=<15 20 20>
      constraintSetup {angularLimitsMax=<10 10 10>, angularLimitsMin=<-5 -5 -5>, angularTarget=<1 0 0>}
      localJointOffset=<0.2 0 0>
     }]
   }
   n3 = Anim Dynamics - Bone: dyn_frontHair_06!(SimulationSpace:AnimPhysSimSpaceType Enum, ComponentPose=@n2, Alpha=1, ExternalForce="0,0,0") -> (Pose) {
    angularDampingOverride=1
    angularSpringConstant=6000
    bAngularSpring=true
    bOverrideAngularDamping=true
    bUseGravityOverride=true
    gravityScale=0.35
    simulationSpace=RootRelative
    alphaBoolBlend {blendOption=Linear}
    boundBone {boneName=dyn_frontHair_06}
    componentLinearAccScale=<1 1 1>
    componentLinearVelScale=<1 1 1>
    gravityOverride {z=-980}
    physicsBodyDefinitions [{
      boundBone {boneName=dyn_frontHair_06}
      boxExtents=<15 20 20>
      constraintSetup {angularLimitsMax=<10 10 10>, angularLimitsMin=<-5 -5 -5>, angularTarget=<1 0 0>}
      localJointOffset=<0.2 0 0>
     }]
   }
   n4 = Anim Dynamics - Bone: dyn_frontHair_03!(SimulationSpace:AnimPhysSimSpaceType Enum, ComponentPose=@n3, Alpha=1, ExternalForce="0,0,0") -> (Pose) {
    angularDampingOverride=1
    angularSpringConstant=6000
    bAngularSpring=true
    bOverrideAngularDamping=true
    bUseGravityOverride=true
    gravityScale=0.35
    simulationSpace=RootRelative
    alphaBoolBlend {blendOption=Linear}
    boundBone {boneName=dyn_frontHair_03}
    componentLinearAccScale=<1 1 1>
    componentLinearVelScale=<1 1 1>
    gravityOverride {z=-980}
    physicsBodyDefinitions [{
      boundBone {boneName=dyn_frontHair_03}
      boxExtents=<15 20 20>
      constraintSetup {angularLimitsMax=<10 10 10>, angularLimitsMin=<-5 -5 -5>, angularTarget=<1 0 0>}
      localJointOffset=<0.2 0 0>
     }]
   }
   n5 = Anim Dynamics - Bone: dyn_frontHair_04!(SimulationSpace:AnimPhysSimSpaceType Enum, ComponentPose=@n4, Alpha=1, ExternalForce="0,0,0") -> (Pose) {
    angularDampingOverride=1
    angularSpringConstant=6000
    bAngularSpring=true
    bOverrideAngularDamping=true
    bUseGravityOverride=true
    gravityScale=0.35
    simulationSpace=RootRelative
    alphaBoolBlend {blendOption=Linear}
    boundBone {boneName=dyn_frontHair_04}
    componentLinearAccScale=<1 1 1>
    componentLinearVelScale=<1 1 1>
    gravityOverride {z=-980}
    physicsBodyDefinitions [{
      boundBone {boneName=dyn_frontHair_04}
      boxExtents=<15 20 20>
      constraintSetup {angularLimitsMax=<10 10 10>, angularLimitsMin=<-5 -5 -5>, angularTarget=<1 0 0>}
      localJointOffset=<0.2 0 0>
     }]
   }
   n6 = Anim Dynamics - Bone: dyn_frontHair_05!(SimulationSpace:AnimPhysSimSpaceType Enum, ComponentPose=@n5, Alpha=1, ExternalForce="0,0,0") -> (Pose) {
    angularDampingOverride=1
    angularSpringConstant=6000
    bAngularSpring=true
    bOverrideAngularDamping=true
    bUseGravityOverride=true
    gravityScale=0.35
    simulationSpace=RootRelative
    alphaBoolBlend {blendOption=Linear}
    boundBone {boneName=dyn_frontHair_05}
    componentLinearAccScale=<1 1 1>
    componentLinearVelScale=<1 1 1>
    gravityOverride {z=-980}
    physicsBodyDefinitions [{
      boundBone {boneName=dyn_frontHair_05}
      boxExtents=<15 20 20>
      constraintSetup {angularLimitsMax=<10 10 10>, angularLimitsMin=<-5 -5 -5>, angularTarget=<1 0 0>}
      localJointOffset=<0.2 0 0>
     }]
   }
   n7 = Anim Dynamics - Bone: dyn_HairSide_l_01!(SimulationSpace:AnimPhysSimSpaceType Enum, ComponentPose=@n6, Alpha=0.4, ExternalForce="0,0,0") -> (Pose) {
    angularDampingOverride=1
    angularSpringConstant=6000
    bAngularSpring=true
    bOverrideAngularDamping=true
    bUseGravityOverride=true
    gravityScale=0.35
    simulationSpace=RootRelative
    alphaBoolBlend {blendOption=Linear}
    boundBone {boneName=dyn_HairSide_l_01}
    componentLinearAccScale=<1 1 1>
    componentLinearVelScale=<1 1 1>
    gravityOverride {z=-980}
    physicsBodyDefinitions [{
      boundBone {boneName=dyn_HairSide_l_01}
      boxExtents=<15 20 20>
      constraintSetup {angularLimitsMax=<10 10 10>, angularLimitsMin=<-5 -5 -5>, angularTarget=<1 0 0>}
      localJointOffset=<0.2 0 0>
     }]
   }
   n8 = Anim Dynamics - Bone: dyn_HairSide_l_02!(SimulationSpace:AnimPhysSimSpaceType Enum, ComponentPose=@n7, Alpha=1, ExternalForce="0,0,0") -> (Pose) {
    angularDampingOverride=1
    angularSpringConstant=6000
    bAngularSpring=true
    bOverrideAngularDamping=true
    bUseGravityOverride=true
    gravityScale=0.35
    simulationSpace=RootRelative
    alphaBoolBlend {blendOption=Linear}
    boundBone {boneName=dyn_HairSide_l_02}
    componentLinearAccScale=<1 1 1>
    componentLinearVelScale=<1 1 1>
    gravityOverride {z=-980}
    physicsBodyDefinitions [{
      boundBone {boneName=dyn_HairSide_l_02}
      boxExtents=<15 20 20>
      constraintSetup {angularLimitsMax=<10 10 10>, angularLimitsMin=<-5 -5 -5>, angularTarget=<1 0 0>}
      localJointOffset=<0.2 0 0>
     }]
   }
   n9 = Anim Dynamics - Bone: dyn_HairSide_l_03!(SimulationSpace:AnimPhysSimSpaceType Enum, ComponentPose=@n8, Alpha=1, ExternalForce="0,0,0") -> (Pose) {
    angularDampingOverride=1
    angularSpringConstant=6000
    bAngularSpring=true
    bOverrideAngularDamping=true
    bUseGravityOverride=true
    gravityScale=0.35
    simulationSpace=RootRelative
    alphaBoolBlend {blendOption=Linear}
    boundBone {boneName=dyn_HairSide_l_03}
    componentLinearAccScale=<1 1 1>
    componentLinearVelScale=<1 1 1>
    gravityOverride {z=-980}
    physicsBodyDefinitions [{
      boundBone {boneName=dyn_HairSide_l_03}
      boxExtents=<15 20 20>
      constraintSetup {angularLimitsMax=<10 10 10>, angularLimitsMin=<-5 -5 -5>, angularTarget=<1 0 0>}
      localJointOffset=<0.2 0 0>
     }]
   }
   n10 = Anim Dynamics - Bone: dyn_HairSide_r_03!(SimulationSpace:AnimPhysSimSpaceType Enum, ComponentPose=@n9, Alpha=0.4, ExternalForce="0,0,0") -> (Pose) {
    angularDampingOverride=1
    angularSpringConstant=6000
    bAngularSpring=true
    bOverrideAngularDamping=true
    bUseGravityOverride=true
    gravityScale=0.35
    simulationSpace=RootRelative
    alphaBoolBlend {blendOption=Linear}
    boundBone {boneName=dyn_HairSide_r_03}
    componentAppliedLinearAccClamp=<10000 10000 10000>
    componentLinearAccScale=<1 1 1>
    componentLinearVelScale=<1 1 1>
    gravityOverride {z=-980}
    physicsBodyDefinitions [{
      boundBone {boneName=dyn_HairSide_r_03}
      boxExtents=<15 20 20>
      constraintSetup {angularLimitsMax=<10 10 10>, angularLimitsMin=<-5 -5 -5>, angularTarget=<1 0 0>}
      localJointOffset=<0.2 0 0>
     }]
   }
   n11 = Anim Dynamics - Bone: dyn_HairSide_r_04!(SimulationSpace:AnimPhysSimSpaceType Enum, ComponentPose=@n10, Alpha=0.6, ExternalForce="0,0,0") -> (Pose) {
    angularDampingOverride=1
    angularSpringConstant=6000
    bAngularSpring=true
    bOverrideAngularDamping=true
    bUseGravityOverride=true
    gravityScale=0.35
    simulationSpace=RootRelative
    alphaBoolBlend {blendOption=Linear}
    boundBone {boneName=dyn_HairSide_r_04}
    componentAppliedLinearAccClamp=<10000 10000 10000>
    componentLinearAccScale=<1 1 1>
    componentLinearVelScale=<1 1 1>
    gravityOverride {z=-980}
    physicsBodyDefinitions [{
      boundBone {boneName=dyn_HairSide_r_04}
      boxExtents=<15 20 20>
      constraintSetup {angularLimitsMax=<10 10 10>, angularLimitsMin=<-5 -5 -5>, angularTarget=<1 0 0>}
      localJointOffset=<0.2 0 0>
     }]
   }
   n12 = Anim Dynamics - Bone: dyn_HairSide_r_05!(SimulationSpace:AnimPhysSimSpaceType Enum, ComponentPose=@n11, Alpha=0.6, ExternalForce="0,0,0") -> (Pose) {
    angularDampingOverride=1
    angularSpringConstant=6000
    bAngularSpring=true
    bOverrideAngularDamping=true
    bUseGravityOverride=true
    gravityScale=0.35
    simulationSpace=RootRelative
    alphaBoolBlend {blendOption=Linear}
    boundBone {boneName=dyn_HairSide_r_05}
    componentAppliedLinearAccClamp=<10000 10000 10000>
    componentLinearAccScale=<1 1 1>
    componentLinearVelScale=<1 1 1>
    gravityOverride {z=-980}
    physicsBodyDefinitions [{
      boundBone {boneName=dyn_HairSide_r_05}
      boxExtents=<15 20 20>
      constraintSetup {angularLimitsMax=<10 10 10>, angularLimitsMin=<-5 -5 -5>, angularTarget=<1 0 0>}
      localJointOffset=<0.2 0 0>
     }]
   }
   n13 = Anim Dynamics - Bone: dyn_HairSide_r_01!(SimulationSpace:AnimPhysSimSpaceType Enum, ComponentPose=@n12, Alpha=0.4, ExternalForce="0,0,0") -> (Pose) {
    angularDampingOverride=1
    angularSpringConstant=6000
    bAngularSpring=true
    bOverrideAngularDamping=true
    bUseGravityOverride=true
    gravityScale=0.35
    simulationSpace=RootRelative
    alphaBoolBlend {blendOption=Linear}
    boundBone {boneName=dyn_HairSide_r_01}
    componentAppliedLinearAccClamp=<10000 10000 10000>
    componentLinearAccScale=<1 1 1>
    componentLinearVelScale=<1 1 1>
    gravityOverride {z=-980}
    physicsBodyDefinitions [{
      boundBone {boneName=dyn_HairSide_r_01}
      boxExtents=<15 20 20>
      constraintSetup {angularLimitsMax=<10 10 10>, angularLimitsMin=<-5 -5 -5>, angularTarget=<1 0 0>}
      localJointOffset=<0.2 0 0>
     }]
   }
   n14 = Anim Dynamics - Bone: dyn_HairSide_r_02!(SimulationSpace:AnimPhysSimSpaceType Enum, ComponentPose=@n13, Alpha=0.8, ExternalForce="0,0,0") -> (Pose) {
    angularDampingOverride=1
    angularSpringConstant=6000
    bAngularSpring=true
    bOverrideAngularDamping=true
    bUseGravityOverride=true
    gravityScale=0.35
    simulationSpace=RootRelative
    alphaBoolBlend {blendOption=Linear}
    boundBone {boneName=dyn_HairSide_r_02}
    componentAppliedLinearAccClamp=<10000 10000 10000>
    componentLinearAccScale=<1 1 1>
    componentLinearVelScale=<1 1 1>
    gravityOverride {z=-980}
    physicsBodyDefinitions [{
      boundBone {boneName=dyn_HairSide_r_02}
      boxExtents=<15 20 20>
      constraintSetup {angularLimitsMax=<10 10 10>, angularLimitsMin=<-5 -5 -5>, angularTarget=<1 0 0>}
      localJointOffset=<0.2 0 0>
     }]
   }
   n15 = Anim Dynamics - Bone: dyn_backHair_03!(SimulationSpace:AnimPhysSimSpaceType Enum, ComponentPose=@n14, Alpha=0.6, ExternalForce="0,0,0") -> (Pose) {
    angularDampingOverride=1
    angularSpringConstant=6000
    bAngularSpring=true
    bOverrideAngularDamping=true
    bUseGravityOverride=true
    gravityScale=0.35
    simulationSpace=RootRelative
    alphaBoolBlend {blendOption=Linear}
    boundBone {boneName=dyn_backHair_03}
    componentLinearAccScale=<1 1 1>
    componentLinearVelScale=<1 1 1>
    gravityOverride {z=-980}
    physicsBodyDefinitions [{
      boundBone {boneName=dyn_backHair_03}
      boxExtents=<15 20 20>
      constraintSetup {angularLimitsMax=<10 10 10>, angularLimitsMin=<-5 -5 -5>, angularTarget=<1 0 0>}
      localJointOffset=<0.2 0 0>
     }]
   }
   n16 = Anim Dynamics - Bone: dyn_backHair_02!(SimulationSpace:AnimPhysSimSpaceType Enum, ComponentPose=@n15, Alpha=0.8, ExternalForce="0,0,0") -> (Pose) {
    angularDampingOverride=1
    angularSpringConstant=6000
    bAngularSpring=true
    bOverrideAngularDamping=true
    bUseGravityOverride=true
    gravityScale=0.35
    simulationSpace=RootRelative
    alphaBoolBlend {blendOption=Linear}
    boundBone {boneName=dyn_backHair_02}
    componentLinearAccScale=<1 1 1>
    componentLinearVelScale=<1 1 1>
    gravityOverride {z=-980}
    physicsBodyDefinitions [{
      boundBone {boneName=dyn_backHair_02}
      boxExtents=<15 20 20>
      constraintSetup {angularLimitsMax=<10 10 10>, angularLimitsMin=<-5 -5 -5>, angularTarget=<1 0 0>}
      localJointOffset=<0.2 0 0>
     }]
   }
   n17 = Anim Dynamics - Bone: dyn_backHair_01!(SimulationSpace:AnimPhysSimSpaceType Enum, ComponentPose=@n16, Alpha=0.4, ExternalForce="0,0,0") -> (Pose) {
    angularDampingOverride=1
    angularSpringConstant=6000
    bAngularSpring=true
    bOverrideAngularDamping=true
    bUseGravityOverride=true
    gravityScale=0.35
    simulationSpace=RootRelative
    alphaBoolBlend {blendOption=Linear}
    boundBone {boneName=dyn_backHair_01}
    componentLinearAccScale=<1 1 1>
    componentLinearVelScale=<1 1 1>
    gravityOverride {z=-980}
    physicsBodyDefinitions [{
      boundBone {boneName=dyn_backHair_01}
      boxExtents=<15 20 20>
      constraintSetup {angularLimitsMax=<10 10 10>, angularLimitsMin=<-5 -5 -5>, angularTarget=<1 0 0>}
      localJointOffset=<0.2 0 0>
     }]
   }
   Output Pose(Result=Component To Local(ComponentPose=@n17)) {layerGroup=None}
   n18 = Anim Dynamics - Bone: dyn_beanie_01!(
    SimulationSpace:AnimPhysSimSpaceType Enum
    ComponentPose=Local To Component(LocalPose=Copy Pose From Mesh() {bUseAttachedParent=true})
    Alpha=0.2
    ExternalForce="0,0,0"
   ) -> (Pose) {
    angularDampingOverride=1
    angularSpringConstant=9000
    bAngularSpring=true
    bOverrideAngularDamping=true
    bUseGravityOverride=true
    gravityScale=0.35
    simulationSpace=RootRelative
    alphaBoolBlend {blendOption=Linear}
    boundBone {boneName=dyn_beanie_01}
    componentAppliedLinearAccClamp=<10000 10000 10000>
    componentLinearAccScale=<0.6 0.6 0.6>
    componentLinearVelScale=<0.3 0.3 0.3>
    gravityOverride {z=-980}
    physicsBodyDefinitions [{
      boundBone {boneName=dyn_beanie_01}
      boxExtents=<12 25 12>
      constraintSetup {angularLimitsMax=<-5 15 10>, angularLimitsMin=<-5 0 -10>, angularTarget=<1 0 0>}
      localJointOffset=<0.1 0 0>
     }]
   }
   n19 = Anim Dynamics - Bone: dyn_beanie_02!(SimulationSpace:AnimPhysSimSpaceType Enum, ComponentPose=@n18, Alpha=0.2, ExternalForce="0,0,0") -> (Pose) {
    angularDampingOverride=1
    angularSpringConstant=9000
    bAngularSpring=true
    bOverrideAngularDamping=true
    bUseGravityOverride=true
    gravityScale=0.35
    simulationSpace=RootRelative
    alphaBoolBlend {blendOption=Linear}
    boundBone {boneName=dyn_beanie_02}
    componentAppliedLinearAccClamp=<10000 10000 10000>
    componentLinearAccScale=<0.6 0.6 0.6>
    componentLinearVelScale=<0.3 0.3 0.3>
    gravityOverride {z=-980}
    physicsBodyDefinitions [{
      boundBone {boneName=dyn_beanie_02}
      boxExtents=<12 25 12>
      constraintSetup {angularLimitsMax=<-5 15 10>, angularLimitsMin=<-5 0 -10>, angularTarget=<1 0 0>}
      localJointOffset=<0.1 0 0>
     }]
   }
   n20 = Anim Dynamics - Bone: dyn_beanie_03!(SimulationSpace:AnimPhysSimSpaceType Enum, ComponentPose=@n19, Alpha=0.2, ExternalForce="0,0,0") -> (Pose) {
    angularDampingOverride=1
    angularSpringConstant=12000
    bAngularSpring=true
    bOverrideAngularDamping=true
    bUseGravityOverride=true
    gravityScale=0.35
    simulationSpace=RootRelative
    alphaBoolBlend {blendOption=Linear}
    boundBone {boneName=dyn_beanie_03}
    componentAppliedLinearAccClamp=<10000 10000 10000>
    componentLinearAccScale=<0.6 0.6 0.6>
    componentLinearVelScale=<0.3 0.3 0.3>
    gravityOverride {z=-980}
    physicsBodyDefinitions [{
      boundBone {boneName=dyn_beanie_03}
      boxExtents=<12 18 12>
      constraintSetup {angularLimitsMax=<-5 15 10>, angularLimitsMin=<-5 0 -10>, angularTarget=<1 0 0>}
      localJointOffset=<3 0 0>
     }]
   }
   n21 = Anim Dynamics - Bone: dyn_beanie_04!(SimulationSpace:AnimPhysSimSpaceType Enum, ComponentPose=@n20, Alpha=0.3, ExternalForce="0,0,0") -> (Pose) {
    angularDampingOverride=1
    angularSpringConstant=12000
    bAngularSpring=true
    bOverrideAngularDamping=true
    bUseGravityOverride=true
    gravityScale=0.35
    simulationSpace=RootRelative
    alphaBoolBlend {blendOption=Linear}
    boundBone {boneName=dyn_beanie_04}
    componentAppliedLinearAccClamp=<10000 10000 10000>
    componentLinearAccScale=<0.6 0.6 0.6>
    componentLinearVelScale=<0.3 0.3 0.3>
    gravityOverride {z=-980}
    physicsBodyDefinitions [{
      boundBone {boneName=dyn_beanie_04}
      boxExtents=<12 18 12>
      constraintSetup {angularLimitsMax=<-5 15 10>, angularLimitsMin=<-5 0 -10>, angularTarget=<1 0 0>}
      localJointOffset=<4 0 0>
     }]
   }
   Save cached pose 'CachePhysics'(Pose=Component To Local(ComponentPose=@n21)) {cacheName=CachePhysics}
  ]}
 EventGraph {nodes [
   Event BlueprintUpdateAnimation()
   > n1 = Sequence()
   > n2 = Gravity Override State Function DEPRECATED(CharacterPartAnimInstance=Self-Reference(), GameTimeInSeconds=GetGameTimeInSeconds(), GravityOverrideParams=Get Gravity Override Params())
   > Set GravityOverride(GravityOverride=@n2)
   n3 = Branch(execute=@n1.then_1, Condition=Get bIsInFrontEnd())
   > Set Sim Space A(Sim Space A:AnimPhysSimSpaceType Enum=World) -> (Output_Get:AnimPhysSimSpaceType Enum)
   Set Sim Space A(execute=@n3.else)
   n4 = Event OnCharacterCustomizationLoaded()
   > Set BodyMesh(BodyMesh=Get Array Item with EFort Custom Part Type(Array=@n4.CustomMeshParts))
  ]}
]
variables [
 BodyMesh {type=SkeletalMeshComponent}
 Gravity Override Params {type=GravityOverrideParamsStruct}
 GravityOverride {type=Vector}
 Sim Space A {type=AnimPhysSimSpaceType}
]
